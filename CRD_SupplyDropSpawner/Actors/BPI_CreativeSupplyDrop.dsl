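; Blueprint interface for creative supply drop actors used with the Supply Drop Spawner V2 device.
; It declares functions to open the drop for an instigating controller, destroy the drop or its balloon, and set its locked state.
; Further setters pass in the parent device, flare color and spawner replication data.
; The file holds only function signatures, with no implementation.
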(fn DestroyBalloon)

(fn DestroyCreativeSupplyDrop)

(fn OpenCreativeSupplyDrop
  (params (Instigator :type Controller%20Object%20Reference)))

(fn SetFlareColor
  (params (FlareColor :type Linear%20Color%20Structure)))

(fn SetLocked
  (params (Locked :type Boolean)))

(fn SetParentDevice
  (params (ParentDevice :type Device%20Supply%20Drop%20Spawner%20V2%20Object%20Reference)))

(fn SetReplicationData
  (params (ReplicationData :type Supply%20Drop%20Spawner%20Replication%20Data%20Structure)))
