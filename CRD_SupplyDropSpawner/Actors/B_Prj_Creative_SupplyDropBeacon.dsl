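; Creative supply drop beacon projectile that explodes when its fuse ends and tries to spawn a CreativeSupplyDrop.
; On authority, an async sphere overlap high above the beacon looks for tagged supply drops, while remotes hide the beacon.
; If clear, it spawns the drop at the beacon's location, sets fall speed and expires a nested child actor.
; If blocked, it shifts the check diagonally and retries while attempts remain.
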(event EventBeginPlay
  (|CallParentFunction) ; This node is disabled and will not be called.\nDrag off pins to build functionality.
)

(event Collision|EventActorBeginOverlap
  (params (OtherActor :type Actor%20Object%20Reference)))

(event EventOnExploded
  (params (HitActors :type Array%20of%20Actor%20Object%20References) (HitResults :type Array%20of%20Hit%20Result%20Structures))
  (|CallParentFunction HitActors HitResults)
  (Networking|SwitchHasAuthority
    (:Authority
      (CallFunction|AttemptSpawn))
    (:Remote
      (Rendering|SetActorHiddenInGame true))))

(event EventTick
  (params (DeltaSeconds :type Float%20%28single-precision%29)))

(fn AttemptSpawn
  (Class|AsyncOverlapQueue|AsyncSphereOverlapActors :Location (Utilities|Operators|Add (Transformation|GetActorLocation) (Math|Vector|MakeVector (Variables|Default|GetSpawnLocationModifier) (Variables|Default|GetSpawnLocationModifier) 2e+04)) :Radius 1024.0 :ShouldCancelProcess (EventDispatchers|CreateEvent :!event ThrottledHitActorsCancelledEvent) :ProcessHitActorFunction (EventDispatchers|CreateEvent :!event ProcessThrottledHitActorsEvent) :CompletionFunction (EventDispatchers|CreateEvent :!event ThrottledHitActorsCompleteEvent) :TraceChannel "ECC_Visibility"))

(fn FuseEnded
  (|CallParentFunction)
  (Delivery|DoExplosion))

(fn ProcessThrottledHitActorsEvent
  (params (HitActor :type Actor%20Object%20Reference))
  (if (Actor|ActorHasTag HitActor "Creative.SupplyDrop")
    (Variables|Default|SetAllowedToSpawn false)))

(fn ThrottledHitActorsCancelledEvent
  (return false))

(fn ThrottledHitActorsCompleteEvent
  (params (WasCancelled :type Boolean))
  (bind _spawnattempts (Variables|Default|GetSpawnAttempts))
  (if (Variables|Default|GetAllowedToSpawn)
    (bind _returnvalue (Game|SpawnActorfromClass "/CRD_SupplyDropSpawner/Actors/CreativeSupplyDrop.CreativeSupplyDrop_C" (Math|Transform|MakeTransform (Transformation|GetActorLocation)) "AdjustIfPossibleButAlwaysSpawn" "OverrideRootScale" 0 "(R=0.221481,G=1.955463,B=3.884724,A=1.000000)" 0 0))
    (Variables|/Game/Athena/SupplyDrops/AthenaSupplyDrop.AthenaSupplyDrop_C|SetFallSpeed 3e+03 _returnvalue)
    (Actor|SetLifeSpan (Variables|/Script/Engine.ChildActorComponent|GetChildActor (Variables|/Game/Athena/SupplyDrops/AthenaSupplyDrop.AthenaSupplyDrop_C|GetChildActor _returnvalue)) 0.1)
    (elif (> _spawnattempts 0)
      (Math|Integer|DecrementInt _spawnattempts
        (:%20%20
          (Variables|Default|SetSpawnLocationModifier (+ (Variables|Default|GetSpawnLocationModifier) 5e+02))
          (Variables|Default|SetAllowedToSpawn true)
          (CallFunction|AttemptSpawn))))))

(fn ConstructionScript
  (|CallParentFunction))
