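; Creative device that spawns and manages a supply drop on a minigame-start delay or a received event, with lock, unlock, open and destroy receivers.
; Picks random points around the device or island, or within the Creative Royale safe zone.
; Obstructed points retry with growing offsets, then fall back to the device location.
; OnRep_SpawnedSupplyDrop applies user options and loot queued from items dropped onto the device outside a running game.
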(event EventBeginPlay)

(event Collision|EventActorBeginOverlap
  (params (OtherActor :type Actor%20Object%20Reference)))

(event EventTick
  (params (DeltaSeconds :type Float%20%28single-precision%29)))

(event MinigameEndedDelegate(FortMinigameProgress)
  (bind _spawnedsupplydrop (Variables|Default|GetSpawnedSupplyDrop))
  ;[1 1 1] Handle Minigame Start/End
  (Utilities|Time|ClearandInvalidateTimerbyHandle (Variables|Default|GetSpawnDelayTimer))
  (Utilities|IsValid _spawnedsupplydrop
    (:Is%20Valid
      (Variables|/CRD_SupplyDropSpawner/Actors/CreativeSupplyDrop.CreativeSupplyDrop_C|SetSpawnLootOnDeath false _spawnedsupplydrop)
      (Actor|DestroyActor _spawnedsupplydrop)
      (CallFunction|ResetSpawnAttemptGate))))

(event MinigameStartedDelegate(FortMinigameProgress)
  ;[1 1 1] Handle Minigame Start/End
  (Networking|SwitchHasAuthority
    (:Authority
      (bind _returnvalue (Utilities|Time|SetTimerbyEvent (EventDispatchers|CreateEvent :!event Custom|SpawnSupplyDrop) (Variables|UserOptions|GetfSpawnDelay)))
      (Variables|Default|SetSpawnDelayTimer _returnvalue)
      (Variables|Default|SetSpawnAttemptsRemaining (Variables|Default|GetSpawnAttempts)))))

(event OnComponentHit(PickupCollisionBox)
  (params (HitComponent :type Primitive%20Component%20Object%20Reference) (OtherActor :type Actor%20Object%20Reference) (OtherComp :type Primitive%20Component%20Object%20Reference) (NormalImpulse :type Vector :default "0, 0, 0") (Hit :type Hit%20Result%20Structure%20%28by%20ref%29))
  ;[1 1 1] Handle Item Drops
  (Networking|SwitchHasAuthority
    (:Authority
      (Utilities|IsValid OtherActor
        (:Is%20Valid
          (if (not (Minigame|HasStarted (Variables|/Script/FortniteGame.FortMinigameLogicComponent|GetCurrentMinigame (Variables|Default|GetFortMinigameProgress)) :!class /Script/FortniteGame.FortMinigame))
            (CallFunction|HandleItemDropped OtherActor)))))))

(event Custom|Multicast_AddItemFX
  (params (Item :type Fort%20Pickup%20Athena%20Object%20Reference))
  ;[1 1 1] Handle Item Drops
  (Networking|CanExecuteCosmeticEvents
    (:True
      (Class|CreativeItemPreviewComponent|ItemAddedToQueueFX (Variables|Default|GetCreative_ItemPreview_Component) Item))))

(event EventInitializeDevice
  ;[1 1 1] Handle Item Drops
  (|CallParentFunction)
  (Initialization|InitializeComponent (Variables|Default|GetCreative_ItemPreview_Component) (Variables|Default|GetPreviewMeshScene) (Variables|Default|GetPreviewStaticMesh) (Variables|Default|GetPreviewSkeletalMesh) 0 (Variables|Default|GetPickupItemListComponent) 0 :!class /Game/Creative/Devices/Common/Components/Creative_ItemPreview_Component.Creative_ItemPreview_Component_C)
  (Initialization|InitializeComponentWithAffects (Variables|Default|GetCreative_UseRestriction_Component) (Variables|Default|GetFortMinigameLogic) 0 0 (Variables|UserOptions|GetOwningTeamIndexSetting) (Variables|UserOptions|GetOwningClassIndexSetting) 0 1 (not (Math|Conversions|ToBoolean(Integer) (Variables|UserOptions|GetKnob_AffectsTeam))) (not (Math|Conversions|ToBoolean(Integer) (Variables|UserOptions|GetKnob_AffectsClass))) false)
  (Initialization|InitializeComponentWithAffects (Variables|Default|GetCreative_UseRestriction_Damage) (Variables|Default|GetFortMinigameLogic) 0 0 (Variables|UserOptions|GetOwningTeamIndexSetting) (Variables|UserOptions|GetOwningClassIndexSetting) 0 1 (not (Math|Conversions|ToBoolean(Integer) (Variables|UserOptions|GetKnob_DamageTeam))) (not (Math|Conversions|ToBoolean(Integer) (Variables|UserOptions|GetKnob_DamageClass))) false)
  (Variables|Default|SetSupplyDropSpawning false))

(event Custom|SpawnSupplyDrop
  (bind (_spawnlocation _bsuccess) (CallFunction|CalculateSpawnLocation))
  (bind _spawnattemptsremaining (Variables|Default|GetSpawnAttemptsRemaining))
  ;[0 1 0.3421 0.5] Spawn Supply Drop
  (if _bsuccess
    (bind _output_get (Variables|Default|SetSpawnLocation _spawnlocation))
    (if (CallFunction|CanSpawnSupplyDrop _output_get)
      _
      (elif (<= _spawnattemptsremaining 0)
        (Utilities|FlowControl|DoOnce false
          (:Completed
            (bind _output_get_2 (Variables|Default|SetSpawnLocation (Transformation|GetActorLocation)))
            (if (CallFunction|CanSpawnSupplyDrop _output_get_2 true)
              (Variables|Default|SetForcedSpawnAtLocation true)
              (bind _returnvalue (Building|SpawnBuildingActor "/CRD_SupplyDropSpawner/Actors/CreativeSupplyDrop.CreativeSupplyDrop_C" (Math|Transform|MakeTransform (Math|Vector|SelectVector (Variables|Default|GetSpawnLocation) (Utilities|Operators|Add (Variables|Default|GetSpawnLocation) (Math|Vector|MakeVector (Variables|Default|GetSpawnLocationModifier) (Variables|Default|GetSpawnLocationModifier))) (Variables|Default|GetForcedSpawnAtLocation))) self (CallFunction|GetSupplyDropInstigatorPawn) false false))
              (bind _ascreative_supply_drop (Utilities|Casting|CastToCreativeSupplyDrop _returnvalue))
              (bind _output_get_3 (Variables|Default|SetSpawnedSupplyDrop _ascreative_supply_drop))
              (Class|CreativeSupplyDrop|SetFlareColor _output_get_3 (Variables|UserOptions|GetSupplyDropFXColor))
              (else
                (Development|PrintString "CRD_SupplyDropSpawner::SpawnSupplyDrop Exhausted all attempts to spawn supply drop." true true "(R=1.000000,G=0.142179,B=0.000000,A=1.000000)" 1e+01)))
            (Variables|Default|SetSpawnLocationModifier 0.0)
            (Variables|Default|SetSpawnAttemptsRemaining (Variables|Default|GetSpawnAttempts))
            (Variables|Default|SetSupplyDropSpawning false)
            (Variables|Default|SetForcedSpawnAtLocation false)
            (CallFunction|ResetSpawnAttemptGate)))
        (else
          (Math|Integer|DecrementInt _spawnattemptsremaining
            (:%20%20
              (Variables|Default|SetSpawnLocationModifier (+ (Variables|Default|GetSpawnLocationModifier) 6.5e+02))
              (Utilities|FlowControl|Delay 1.0)
              (CallFunction|SpawnSupplyDrop))))))
    (else
      (Development|PrintString "CRD_SupplyDropSpawner::SpawnSupplyDrop Error: Unable to find valid spawn location." true true "(R=1.000000,G=0.000000,B=0.069327,A=1.000000)" 1e+01))))

(event Custom|ResetSpawnAttemptGate
  ;[0 1 0.3421 0.5] Spawn Supply Drop
  (Utilities|FlowControl|DoOnce false
    (:Completed
      (bind _output_get (Variables|Default|SetSpawnLocation (Transformation|GetActorLocation)))
      (if (CallFunction|CanSpawnSupplyDrop _output_get true)
        (Variables|Default|SetForcedSpawnAtLocation true)
        (bind _returnvalue (Building|SpawnBuildingActor "/CRD_SupplyDropSpawner/Actors/CreativeSupplyDrop.CreativeSupplyDrop_C" (Math|Transform|MakeTransform (Math|Vector|SelectVector (Variables|Default|GetSpawnLocation) (Utilities|Operators|Add (Variables|Default|GetSpawnLocation) (Math|Vector|MakeVector (Variables|Default|GetSpawnLocationModifier) (Variables|Default|GetSpawnLocationModifier))) (Variables|Default|GetForcedSpawnAtLocation))) self (CallFunction|GetSupplyDropInstigatorPawn) false false))
        (bind _ascreative_supply_drop (Utilities|Casting|CastToCreativeSupplyDrop _returnvalue))
        (bind _output_get_2 (Variables|Default|SetSpawnedSupplyDrop _ascreative_supply_drop))
        (Class|CreativeSupplyDrop|SetFlareColor _output_get_2 (Variables|UserOptions|GetSupplyDropFXColor))
        (else
          (Development|PrintString "CRD_SupplyDropSpawner::SpawnSupplyDrop Exhausted all attempts to spawn supply drop." true true "(R=1.000000,G=0.142179,B=0.000000,A=1.000000)" 1e+01)))
      (Variables|Default|SetSpawnLocationModifier 0.0)
      (Variables|Default|SetSpawnAttemptsRemaining (Variables|Default|GetSpawnAttempts))
      (Variables|Default|SetSupplyDropSpawning false)
      (Variables|Default|SetForcedSpawnAtLocation false)
      (CallFunction|ResetSpawnAttemptGate))))

(event Custom|SpawnWhenReceivingEvent
  (params (Instigator :type Controller%20Object%20Reference))
  ;[0.1659 0.0612 0.4896] Triggers and Recievers
  (if (not (Variables|Default|GetSupplyDropSpawning))
    (Variables|Default|SetSupplyDropSpawning true)
    (Utilities|IsValid (Variables|Default|GetSpawnedSupplyDrop)
      (:Is%20Valid
        (Variables|Default|SetSupplyDropSpawning false))
      (:Is%20Not%20Valid
        (Variables|Default|SetSupplyDropSpawnerInstigator Instigator)
        (Utilities|Time|ClearandInvalidateTimerbyHandle (Variables|Default|GetSpawnDelayTimer))
        (CallFunction|SpawnSupplyDrop)))))

(event Custom|DestroyWhenReceivingEvent
  (params (Instigator :type Controller%20Object%20Reference))
  (bind _spawnedsupplydrop (Variables|Default|GetSpawnedSupplyDrop))
  ;[0.1659 0.0612 0.4896] Triggers and Recievers
  (Utilities|IsValid _spawnedsupplydrop
    (:Is%20Valid
      (if (not (Variables|/Game/Athena/SupplyDrops/AthenaSupplyDrop.AthenaSupplyDrop_C|GetBalloonPopped _spawnedsupplydrop))
        (bind _ascreative_supply_drop_balloon (Utilities|Casting|CastToCreativeSupplyDropBalloon (Variables|/Script/Engine.ChildActorComponent|GetChildActor (Variables|/Game/Athena/SupplyDrops/AthenaSupplyDrop.AthenaSupplyDrop_C|GetChildActor _spawnedsupplydrop))))
        (Actor|SetLifeSpan _ascreative_supply_drop_balloon 0.1)))))

(event Custom|OpenSupplyDropWhenReceivingEvent
  (params (Instigator :type Controller%20Object%20Reference))
  (bind _spawnedsupplydrop (Variables|Default|GetSpawnedSupplyDrop))
  ;[0.1659 0.0612 0.4896] Triggers and Recievers
  (Utilities|IsValid _spawnedsupplydrop
    (:Is%20Valid
      (Class|AthenaSupplyDrop|OpenSupplyDrop _spawnedsupplydrop)
      (GameplayEvent|BroadcastEvent (Variables|UserOptions|Events|GetOn%20Opened) 0 (Utilities|Struct|MakeGameplayEventSupplyDropSpawnerTransmitonOpen Instigator)))))

(event Custom|UnlockWhenReceivingEvent
  (params (Instigator :type Controller%20Object%20Reference))
  ;[0.1659 0.0612 0.4896] Triggers and Recievers
  (Variables|/CRD_SupplyDropSpawner/Actors/CreativeSupplyDrop.CreativeSupplyDrop_C|SetbLocked false (Variables|Default|GetSpawnedSupplyDrop)))

(event Custom|LockWhenReceivingEvent
  (params (Instigator :type Controller%20Object%20Reference))
  ;[0.1659 0.0612 0.4896] Triggers and Recievers
  (Variables|/CRD_SupplyDropSpawner/Actors/CreativeSupplyDrop.CreativeSupplyDrop_C|SetbLocked true (Variables|Default|GetSpawnedSupplyDrop)))

(event Custom|SetSupplyDropMesh
  (bind _selectedballoonindex (Variables|UserOptions|GetSelectedBalloonIndex))
  (bind _spawnedsupplydrop (Variables|Default|GetSpawnedSupplyDrop))
  (bind _cratestaticmesh (Variables|/Game/Athena/SupplyDrops/AthenaSupplyDrop.AthenaSupplyDrop_C|GetCrateStaticMesh _spawnedsupplydrop))
  (bind _supplydropballoon (Variables|Default|GetSupplyDropBalloon))
  (bind _staticmeshcomponent (Variables|/Script/FortniteGame.BuildingSMActor|GetStaticMeshComponent _supplydropballoon))
  ;[1 1 1] Cosmetic
  (if (!= _selectedballoonindex 0)
    (bind _object (Utilities|AsyncLoadAsset (Utilities|Array|Get(acopy) (Variables|Default|GetBalloonMeshes) _selectedballoonindex))
      (:then
        (bind _object_2 (Utilities|AsyncLoadAsset (Utilities|Array|Get(acopy) (Variables|Default|GetCrateMeshes) (Variables|UserOptions|GetSelectedBalloonIndex)))
          (:Completed
            (bind _asstatic_mesh (Utilities|Casting|CastToStaticMesh _object_2))
            (Components|StaticMesh|SetStaticMesh _cratestaticmesh _asstatic_mesh)
            ;;[1 1 1] We need to adjust the scale for the birthday version of the supply drop
            (switch int (Variables|UserOptions|GetSelectedBalloonIndex)
              (:2
                (Transformation|SetRelativeScale3D _cratestaticmesh "2.000000,2.000000,2.000000")
                (Transformation|SetRelativeScale3D (Variables|/Script/FortniteGame.BuildingSMActor|GetStaticMeshComponent (Variables|Default|GetSupplyDropBalloon)) "0.420000,0.420000,0.420000"))))))
      (:Completed
        (bind _asstatic_mesh_2 (Utilities|Casting|CastToStaticMesh _object))
        (Components|StaticMesh|SetStaticMesh _staticmeshcomponent _asstatic_mesh_2)
        (Transformation|SetRelativeScale3D _staticmeshcomponent "0.800000,0.800000,0.800000")))))

(fn AttemptSupplyDropSpawn
  (bind _returnvalue (Utility|OverlapQueue|GetAsyncOverlapQueue self))
  (Class|AsyncOverlapQueue|ThrottledAsyncSphereOverlapActors :self _returnvalue :Location (Utilities|Operators|Add (Transformation|GetActorLocation) (Math|Vector|MakeVector (Variables|Default|GetSpawnLocationModifier) (Variables|Default|GetSpawnLocationModifier) (Variables|Default|GetFallHeight))) :Radius 1e+04 :bCheckLineOfSight false :ShouldCancelProcess (EventDispatchers|CreateEvent :!event ThrottleSpawnCancelEvent) :ProcessHitActorFunction (EventDispatchers|CreateEvent :!event ThrottleSpawnProcessEvent) :CompletionFunction (EventDispatchers|CreateEvent :!event ThrottleSpawnCompletionEvent) :ObjectTypes "ObjectTypeQuery1" :TraceChannel "ECC_Visibility"))

(fn CalculateSpawnLocation
  (bind (_safezonelocation _bsuccess) (CallFunction|MakeSupplyDropLocationfromSafeZone))
  (bind _spawnradius (CallFunction|GetSpawnRadius))
  (bind _returnvalue (* _spawnradius -1.0))
  (bind _returnvalue_2 (Math|Random|RandomFloatinRange _returnvalue _spawnradius))
  ;[0 1 0.2245 0.5] Calculate the spawn location. Use the safe zone if Creative Royale.
  (if (CallFunction|IsCreativeRoyale)
    (return _safezonelocation _bsuccess)
    (else
      ;;[1 0.8798 0.5885 0.5] Save as local variables to force the calculation to run twice.
      (LocalVariables|SetRadomLocXLocal _returnvalue_2)
      (LocalVariables|SetRadomLocYLocal _returnvalue_2)
      (return (Utilities|Operators|Add (CallFunction|GetOriginforCalculations) (Math|Vector|MakeVector (LocalVariables|GetRadomLocXLocal) (LocalVariables|GetRadomLocYLocal))) true))))

(fn CanSpawnSupplyDrop
  (params (Start :type Vector) (ForceSpawnAtDeviceLocation :type Boolean))
  (bind _returnvalue (BattleRoyaleGamePhaseLogic|GetBattleRoyaleGamePhaseLogicComponent))
  (bind _safezoneindicator (CallFunction|GetSafeZoneIndicator))
  (bind _adjustedstartlocationlocal (LocalVariables|GetAdjustedStartLocationLocal))
  (LocalVariables|SetAdjustedStartLocationLocal (Math|Vector|SelectVector (Utilities|Operators|Add Start (Math|Vector|MakeVector 0.0 0.0 (Variables|Default|GetFallHeight))) (Utilities|Operators|Add Start (Math|Vector|MakeVector (Variables|Default|GetSpawnLocationModifier) (Variables|Default|GetSpawnLocationModifier) (Variables|Default|GetFallHeight))) ForceSpawnAtDeviceLocation))
  (Utilities|FlowControl|Sequence
    (:then_0
      ;[0 1 0.2869 0.5] If there is a safe zone indicator (normal Creative does not have one), check that the safe zone radius is greater than zero and check if the modified start location is within the current safe zone.
      (Utilities|IsValid _safezoneindicator
        (:Is%20Valid
          (if (> (SafeZone|GetSafeZoneRadius _safezoneindicator) 0)
            (Utilities|IsValid _returnvalue
              (:Is%20Valid
                (if (not (BattleRoyaleGamePhaseLogic|IsinCurrentSafeZone _returnvalue (LocalVariables|GetAdjustedStartLocationLocal) false false))
                  _))
              (:Is%20Not%20Valid
                (Development|PrintString "CRD_SupplyDropSpawner::CanSpawnSupplyDrop BR GamePhaseLogic component is invalid." true true "(R=1.000000,G=0.142179,B=0.000000,A=1.000000)" 1e+01)))
            (else
              (Development|PrintString "CRD_SupplyDropSpawner::CanSpawnSupplyDrop Safe zone radius is NOT greater than 0." true true "(R=1.000000,G=0.142179,B=0.000000,A=1.000000)" 1e+01)))
          (return false))))
    (:then_1
      ;[1 0.7986 0.3802 0.5] Trace to check for obstructions.
      (Utilities|IsLocationRelevanttoMe _adjustedstartlocationlocal
        (:IsRelevant
          (bind (_outhits _returnvalue_2) (Collision|MultiSphereTraceByChannel _adjustedstartlocationlocal _adjustedstartlocationlocal 1024.0))
          (return (not _returnvalue_2)))
        (:NotRelevant
          (return false))))))

(fn ClearSpawnedSupplyDrop
  (Variables|Default|SetSpawnedSupplyDrop 0))

(fn GetMinigameVolumeMesh
  (bind _self self)
  (bind _returnvalue (Minigame|GetMinigame _self :!class /Script/FortniteGame.MinigameBlueprintLibrary))
  ;[0 1 0.2101 0.5] Get the Minigame volume's mesh.
  (Utilities|IsValid _returnvalue
    (:Is%20Valid
      (bind _asfort_volume (Utilities|Casting|CastToFortVolume (Variables|/Script/FortniteGame.FortMinigame|GetVolume _returnvalue))
        (:then
          (return (Volume|GetStaticMesh _asfort_volume)))
        (:CastFailed
          (Development|PrintString "CRD_SupplyDropSpawner::GetMinigameVolumeMesh Error: Cast to FortVolume failed." true true "(R=1.000000,G=0.015783,B=0.000000,A=1.000000)" 1e+01)
          (return 0))))
    (:Is%20Not%20Valid
      (Development|PrintString "CRD_SupplyDropSpawner::GetMinigameVolumeMesh Error: Miningame is NOT valid." true true "(R=1.000000,G=0.015783,B=0.000000,A=1.000000)" 1e+01)
      (return 0))))

(fn GetOriginForCalculations
  (bind _minigamevolumemesh (CallFunction|GetMinigameVolumeMesh))
  (bind _origin (Collision|GetComponentBounds _minigamevolumemesh))
  (bind (_origin_x _origin_y) (Math|Vector|BreakVector _origin))
  (bind _returnvalue (Transformation|GetActorLocation))
  (bind (_returnvalue_x _returnvalue_y _returnvalue_z) (Math|Vector|BreakVector (Transformation|GetActorLocation)))
  ;[0 1 0.3424 0.5] If the Spawn Radius is -1, use the Minigame volume mesh origin; otherwise, use the actor's location as the origin.
  (if (== (Variables|UserOptions|GetfSpawnRadius) -1.0)
    (Utilities|IsValid _minigamevolumemesh
      (:Is%20Valid
        (return _origin_x _origin_y _returnvalue_z))
      (:Is%20Not%20Valid
        (Development|PrintString "CRD_SupplyDropSpawner::SpawnSupplyDrop Error: Miningame Volume Mesh is NOT valid." true true "(R=1.000000,G=0.015783,B=0.000000,A=1.000000)" 1e+01)
        (return _returnvalue)))
    (else
      (return _returnvalue))))

(fn GetPlayableIslandRadius
  (bind _safezoneindicator (CallFunction|GetSafeZoneIndicator))
  (bind _minigamevolumemesh (CallFunction|GetMinigameVolumeMesh))
  (bind (_origin _boxextent) (Collision|GetComponentBounds _minigamevolumemesh))
  (bind (_boxextent_x _boxextent_y) (Math|Vector|BreakVector _boxextent))
  ;[0 1 0.2829 0.5] Gets the radius of the playable region of the map.
  (if (CallFunction|IsCreativeRoyale)
    ;;[1 0.9818 0.3646 0.5] Use the safe zone for Creative Royale.
    (Utilities|IsValid _safezoneindicator
      (:Is%20Valid
        (return (SafeZone|GetSafeZoneRadius _safezoneindicator)))
      (:Is%20Not%20Valid
        (Development|PrintString "CRD_SupplyDropSpawner::GetIslandRadus Error: SafeZoneIndicator is NOT valid." true true "(R=1.000000,G=0.015783,B=0.000000,A=1.000000)" 1e+01)
        (return 0.0)))
    (else
      ;;[1 0.7919 0 0.5] Get the minimum extent of the Minigame's volume mesh for non-Creative Royale maps since there is no safe zone.
      (Utilities|IsValid _minigamevolumemesh
        (:Is%20Valid
          (return (Math|Float|Min(Float) _boxextent_x _boxextent_y)))
        (:Is%20Not%20Valid
          (Development|PrintString "CRD_SupplyDropSpawner::GetIslandRadus Error: Miningame Volume Mesh is NOT valid." true true "(R=1.000000,G=0.015783,B=0.000000,A=1.000000)" 1e+01)
          (return 0.0))))))

(fn GetSafeZoneCenterFromSafeZonePhases
  (bind _returnvalue (SafeZone|GetSafeZoneIndicator))
  (bind (_outsafezonephase _returnvalue_2) (SafeZone|GetPhaseInfo _returnvalue -1))
  ;[0 1 0.02015 0.5] Gets the current safe zone center. Getting the safe zone center by calling GetSafeZoneCenter from the SafeZoneIndicator was returning the mesh's location which has a different Z value that is too high. The high Z value was making it so the supply drop could not detect the ground and would drift toward map center.
  (Utilities|FlowControl|Sequence
    (:then_0
      (Utilities|IsValid _returnvalue
        (:Is%20Valid
          (if _returnvalue_2
            (return (Utilities|Struct|BreakFortSafeZonePhaseInfo _outsafezonephase))
            (else
              (Development|PrintString "CRD_SupplyDropSpawner::GetSafeZoneCenterFromSafeZonePhases Error: GetPhaseInfo failed." true true "(R=1.000000,G=0.022075,B=0.000000,A=1.000000)" 5.0))))
        (:Is%20Not%20Valid
          (Development|PrintString "CRD_SupplyDropSpawner::GetSafeZoneCenterFromSafeZonePhases Error: Safe zone indicator is NOT valid." true true "(R=1.000000,G=0.022075,B=0.000000,A=1.000000)" 5.0))))
    (:then_1
      (Development|PrintString "CRD_SupplyDropSpawner::GetSafeZoneCenterFromSafeZonePhases Error: Unable to get safe zone center." true true "(R=1.000000,G=0.022075,B=0.000000,A=1.000000)" 5.0)
      (return "0, 0, 0"))))

(fn GetSafeZoneIndicator
  (bind _returnvalue (SafeZone|GetSafeZoneIndicator))
  ;[0 1 0.1947 0.5] Get the safe zone indicator.
  (Utilities|IsValid _returnvalue
    (:Is%20Valid
      (return _returnvalue))
    (:Is%20Not%20Valid
      (return 0))))

(fn GetSpawnRadius
  ;[0 1 0.1877 0.5] If the Spawn Radius is -1, use the playable region of the island; otherwise, use the variable's value.
  (return (select (== (Variables|UserOptions|GetfSpawnRadius) -1.0) (CallFunction|GetPlayableIslandRadius) (Variables|UserOptions|GetfSpawnRadius))))

(fn GetSupplyDropInstigatorPawn
  ;[0 1 0.3331 0.5] Gets the controlled pawn from the instigator.
  (bind _supplydropspawnerinstigator (Variables|Default|GetSupplyDropSpawnerInstigator)
    (:then
      (return (Pawn|GetControlledPawn _supplydropspawnerinstigator)))
    (:else
      (return 0))))

(fn HandleItemDropped
  (params (Actor :type Actor%20Object%20Reference))
  (bind _asfort_pickup_athena (Utilities|Casting|CastToFortPickupAthena Actor))
  (bind _output_get (LocalVariables|SetPickupAthena _asfort_pickup_athena))
  (Utilities|IsPickupValid _output_get (Variables|/Game/Creative/Devices/CaptureArea/PickupItemListComponent.PickupItemListComponent_C|GetInvalidItemTags (Variables|Default|GetPickupItemListComponent))
    (:IsValid
      (if (not (Class|PickupItemListComponent|IsIteminQueue (Variables|Default|GetPickupItemListComponent) (LocalVariables|GetPickupAthena)))
        (bind _bitemadded (Class|PickupItemListComponent|AddItemtoQueue (Variables|Default|GetPickupItemListComponent) (LocalVariables|GetPickupAthena)))
        (if _bitemadded
          (CallFunction|MulticastAddItemFX (LocalVariables|GetPickupAthena)))))))

(fn IsCreativeRoyale
  (bind _returnvalue (Fortnite|Athena|GameState|GetGameStateAthena))
  ;[0 1 0.2074 0.5] If this playlist has a Creative Royale tag return true.
  (Utilities|IsValid _returnvalue
    (:Is%20Valid
      (return (GameplayTags|HasTag (Class|FortGameStateAthena|GetAthenaPlaylistContextTags _returnvalue) "(TagName=\"Athena.Playlist.CreativeRoyale\")")))
    (:Is%20Not%20Valid
      (return false))))

(fn MakeSupplyDropLocationFromSafeZone
  (bind _returnvalue (Fortnite|Athena|GameState|GetGameStateAthena))
  (bind _mapinfo (Variables|/Script/FortniteGame.FortGameStateAthena|GetMapInfo _returnvalue))
  ;[0 1 0.2284 0.5] Creates a supply drop spawn location using the safe zone as the origin. Water will be avoided.
  (Utilities|IsValid _returnvalue
    (:Is%20Valid
      (Utilities|IsValid _mapinfo
        (:Is%20Valid
          (bind _center (CallFunction|GetSafeZoneCenterfromSafeZonePhases))
          (return (Class|FortAthenaMapInfo|PickSupplyDropLocation _mapinfo _center (CallFunction|GetSpawnRadius) true (ScalableFloat|GetValueatLevel (Variables|/Script/FortniteGame.FortAthenaMapInfo|GetSupplyDropMaxPlacementHeight _mapinfo)) (ScalableFloat|GetValueatLevel (Variables|/Script/FortniteGame.FortAthenaMapInfo|GetSupplyDropMinPlacementHeight _mapinfo)) false) true))
        (:Is%20Not%20Valid
          _)))
    (:Is%20Not%20Valid
      _))
  (return "0, 0, 0" false))

(fn OnRep_SpawnedSupplyDrop
  (Utilities|IsValid (Variables|Default|GetSpawnedSupplyDrop)
    (:Is%20Valid
      (CallFunction|SetSupplyDropSettings))))

(fn SetBalloonMaxHealth
  (bind _returnvalue (Ability|GetAbilitySystemComponent (Variables|Default|GetSupplyDropBalloon)))
  (bind _supplydropballoon (Variables|Default|GetSupplyDropBalloon))
  (Networking|SwitchHasAuthority
    (:Authority
      (if (== (Variables|UserOptions|GetfBalloonHealth) (Variables|Constants|GetInfiniteHealthIndex))
        (LocalVariables|SetLocalNewMaxHealth 1.0)
        (Fortnite|SetCanbeDamaged _supplydropballoon)
        (else
          (LocalVariables|SetLocalNewMaxHealth (Variables|UserOptions|GetfBalloonHealth))))
      (bind _returnvalue_2 (Ability|GameplayEffect|AssignTagSetbyCallerMagnitude (GameplayEffects|MakeOutgoingSpec :self (Ability|GetAbilitySystemComponent _supplydropballoon) :GameplayEffectClass (Variables|Constants|GetModifyMaxHealth) :Level 0.0) (Variables|Default|GetSetMaxHealthDataTag) (LocalVariables|GetLocalNewMaxHealth)))
      (Utilities|IsValid _returnvalue
        (:Is%20Valid
          (GameplayEffects|RemoveActiveGameplayEffect _returnvalue (Variables|Default|GetMaxHealthEffectHandlerSupplyDrop) -1)
          (bind _returnvalue_3 (GameplayEffects|ApplyGameplayEffectSpecToSelf _returnvalue _returnvalue_2))
          (Variables|Default|SetMaxHealthEffectHandlerSupplyDrop _returnvalue_3))))))

(fn SetFallSpeed
  (bind _ffallspeedmultiplier (Variables|UserOptions|GetfFallSpeedMultiplier))
  (if (!= 0 _ffallspeedmultiplier)
    (Variables|/Game/Athena/SupplyDrops/AthenaSupplyDrop.AthenaSupplyDrop_C|SetFallSpeed (* (Variables|/Game/Athena/SupplyDrops/AthenaSupplyDrop.AthenaSupplyDrop_C|GetFallSpeed (Variables|Default|GetSpawnedSupplyDrop)) _ffallspeedmultiplier) (Variables|Default|GetSpawnedSupplyDrop))
    (else
      (Actor|SetLifeSpan (Variables|Default|GetSupplyDropBalloon) 0.1))))

(fn SetSupplyDropLoot
  (bind _pickupitemlistcomponent (Variables|Default|GetPickupItemListComponent))
  (bind _returnvalue (ItemData|GetContainedItems _pickupitemlistcomponent))
  (Networking|SwitchHasAuthority
    (:Authority
      (if (Utilities|Array|IsNotEmpty _returnvalue)
        (Variables|/CRD_SupplyDropSpawner/Actors/CreativeSupplyDrop.CreativeSupplyDrop_C|SetItemsToSpawn _returnvalue (Variables|Default|GetSpawnedSupplyDrop))))))

(fn SetSupplyDropSettings
  (bind _showsmokefx (Variables|UserOptions|GetShowSmokeFX))
  (bind _returnvalue (not _showsmokefx))
  (bind _spawnedsupplydrop (Variables|Default|GetSpawnedSupplyDrop))
  (Variables|/CRD_SupplyDropSpawner/Actors/CreativeSupplyDrop.CreativeSupplyDrop_C|SetParentDevice self _spawnedsupplydrop)
  (Variables|/CRD_SupplyDropSpawner/Actors/CreativeSupplyDrop.CreativeSupplyDrop_C|SetFortMinigameProgress (Variables|Default|GetFortMinigameProgress) _spawnedsupplydrop)
  (bind _ascreative_supply_drop_balloon (Utilities|Casting|CastToCreativeSupplyDropBalloon (Variables|/Script/Engine.ChildActorComponent|GetChildActor (Variables|/Game/Athena/SupplyDrops/AthenaSupplyDrop.AthenaSupplyDrop_C|GetChildActor _spawnedsupplydrop))))
  (Variables|Default|SetSupplyDropBalloon _ascreative_supply_drop_balloon)
  (CallFunction|SetBalloonMaxHealth)
  (CallFunction|SetFallSpeed)
  (Variables|/CRD_SupplyDropSpawner/Actors/CreativeSupplyDrop.CreativeSupplyDrop_C|SetbLocked (Variables|UserOptions|GetbStartLocked) _spawnedsupplydrop)
  (CallFunction|SetSupplyDropLoot)
  ;[1 1 1] Hide Smoke VFX and Disable Smoke Audio if VFX are set to hidden
  (Development|SetHiddeninGame (Variables|/Game/Athena/SupplyDrops/AthenaSupplyDrop.AthenaSupplyDrop_C|GetReticle_Niagara _spawnedsupplydrop) _returnvalue true)
  (Class|CreativeSupplyDrop|SetAudioLevel _spawnedsupplydrop _returnvalue)
  (MiniMapData|SetMiniMapIndicatorIsVisible (Variables|/CRD_SupplyDropSpawner/Actors/CreativeSupplyDrop.CreativeSupplyDrop_C|GetFortMiniMap _spawnedsupplydrop) (Variables|UserOptions|GetbShowIconOnMinimap))
  (Variables|/CRD_SupplyDropSpawner/Actors/CreativeSupplyDrop.CreativeSupplyDrop_C|SetRestrictionComponent (Variables|Default|GetCreative_UseRestriction_Component) _spawnedsupplydrop)
  (Variables|/CRD_SupplyDropSpawner/Actors/CreativeSupplyDrop.CreativeSupplyDrop_C|SetDamageRestriction (Variables|Default|GetCreative_UseRestriction_Damage) _spawnedsupplydrop)
  (if (!= (Variables|UserOptions|GetfFallSpeedMultiplier) 0)
    (CallFunction|SetSupplyDropMesh)
    (else
      _))
  (Variables|/CRD_SupplyDropSpawner/Actors/CreativeSupplyDropBalloon.CreativeSupplyDropBalloon_C|SetDamageRestriction (Variables|Default|GetCreative_UseRestriction_Damage) (Variables|Default|GetSupplyDropBalloon)))

(fn ThrottleSpawnCancelEvent
  (return false))

(fn ThrottleSpawnCompletionEvent
  (params (WasCancelled :type Boolean))
  (bind _spawnattempts (Variables|Default|GetSpawnAttempts))
  (if (Variables|Default|GetAllowedToSpawn)
    (CallFunction|SpawnSupplyDrop)
    (elif (> _spawnattempts 0)
      (Math|Integer|DecrementInt _spawnattempts
        (:%20%20
          (Variables|Default|SetSpawnLocationModifier (+ (Variables|Default|GetSpawnLocationModifier) 5e+02))
          (Variables|Default|SetAllowedToSpawn true)
          (CallFunction|AttemptSupplyDropSpawn))))))

(fn ThrottleSpawnProcessEvent
  (params (HitActor :type Actor%20Object%20Reference))
  (if (Actor|ActorHasTag HitActor "Creative.SupplyDrop")
    (Variables|Default|SetAllowedToSpawn false)))

(fn UpdateUseRestrictionWithInstigator
  (params (InInstigator :type Actor%20Object%20Reference))
  ;[0 1 0.09829 0.5] If the instigator should be the owner, call SetTeamsWithAffects with the instigator's team. This makes it so only members of the instigator's team can open supply drops.
  (Utilities|IsValid InInstigator
    (:Is%20Valid
      (if (== (Variables|UserOptions|GetOwningTeamIndexSetting) (Variables|Default|GetInstigatorTeamIndexConst))
        (TeamRestrictions|SetTeamswithAffects (Variables|Default|GetCreative_UseRestriction_Component) (Knobs|Team|ReshiftTeamIndexValue (Math|Conversions|ToInteger(Byte) (Team|GetActorPvPTeamIndex InInstigator))) (not (Math|Conversions|ToBoolean(Integer) (Variables|UserOptions|GetKnob_AffectsTeam))))))
    (:Is%20Not%20Valid
      (Development|PrintString "CRD_SupplyDropSpawner::UpdateUseRestrictionWithInstigator Warning: Instigator is NOT valid." true true "(R=1.000000,G=0.788867,B=0.000000,A=1.000000)" 1e+01))))

(fn ConstructionScript
  (|CallParentFunction))
