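; Mini creative supply drop pod for the Supply Drop Spawner device: it lands, can be opened or destroyed, and spawns loot.
; Replicated spawner data drives team/class interaction and damage restrictions, lock state, items, fall speed, reticle and minimap icon.
; Opening, landing and death report back to the parent device; disallowed or locked-pod damage is healed back.
; Custom loot spawns one pickup per tick, otherwise the parent class handles it.
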(event EventBeginPlay
  (|CallParentFunction) ; This node is disabled and will not be called.\nDrag off pins to build functionality.
)

(event Collision|EventActorBeginOverlap
  (params (OtherActor :type Actor%20Object%20Reference)))

(event EventTick
  (params (DeltaSeconds :type Float%20%28single-precision%29)))

(event EventSetFlareColor
  (params (FlareColor :type Linear%20Color%20Structure))
  ; BPI_CreativeSupplyDrop Interface
  (Variables|Default|SetFXColor FlareColor))

(event EventSetParentDevice
  (params (ParentDevice :type Device%20Supply%20Drop%20Spawner%20V2%20Object%20Reference))
  ; BPI_CreativeSupplyDrop Interface
  (Variables|Default|SetParentDevice ParentDevice))

(event EventSetReplicationData
  (params (ReplicationData :type Supply%20Drop%20Spawner%20Replication%20Data%20Structure))
  ; BPI_CreativeSupplyDrop Interface
  (Variables|Default|SetSupplyDropSpawnerReplicationData ReplicationData))

(event EventSetLocked
  (params (Locked :type Boolean))
  ; BPI_CreativeSupplyDrop Interface
  (Variables|Default|SetbLocked Locked))

(event EventOpenCreativeSupplyDrop
  (params (Instigator :type Controller%20Object%20Reference))
  ; BPI_CreativeSupplyDrop Interface
  (CallFunction|OpenSupplyDrop))

(event EventDestroyCreativeSupplyDrop
  ; BPI_CreativeSupplyDrop Interface
  (Variables|Default|SetSpawnLootOnDeath false)
  (Actor|DestroyActor))

(event Interaction|EventBlueprintOnInteract
  (params (InteractingPawn :type Fort%20Pawn%20Object%20Reference) (InteractionBeingAttempted :type EInteractionBeingAttempted%20Enum :default "FirstInteraction"))
  (Utilities|FlowControl|Sequence
    (:then_0
      (bind _output_get (Variables|Default|SetMyInteractingPawn InteractingPawn))
      (|CallParentFunction _output_get InteractionBeingAttempted)
      (Utilities|IsValid (Variables|Default|GetParentDevice)
        (:Is%20Valid
          (GameplayEvent|BroadcastEvent (Variables|/CRD_SupplyDropSpawner/Actors/Device_SupplyDropSpawner_V2.Device_SupplyDropSpawner_V2_C|GetOn%20Opened (Variables|Default|GetParentDevice)) (Variables|Default|GetParentDevice) (Utilities|Struct|MakeGameplayEventSupplyDropSpawnerTransmitonOpen (Pawn|GetController _output_get))))))
    (:then_1
      (Variables|Default|SetMyInteractingPawn 0))))

(event Custom|DelaySpawnLootFromItemComponent
  (bind (_returnvalue_pickuptospawn _returnvalue_pickupquantity) (Utilities|Struct|BreakMinigameSpawnerSpawnParams (ItemData|MakeSpawnParams (Variables|/CRD_SupplyDropSpawner/Actors/Device_SupplyDropSpawner_V2.Device_SupplyDropSpawner_V2_C|GetPickupItemListComponent (Variables|Default|GetParentDevice)) (Variables|Default|GetItemSpawnCount) (Variables|Default|GetSpawn%20Transform))))
  (bind _itemstospawn (Variables|Default|GetItemsToSpawn))
  (bind _itemstospawn_2 (Variables|Default|GetItemsToSpawn))
  ; Spawn Loot from ItemsToSpawn
  (if (Utilities|Array|IsNotEmpty _itemstospawn)
    (Inventory|SpawnPickupInWorld (Utilities|Array|Get(acopy) _itemstospawn) _returnvalue_pickupquantity (CallFunction|GetLootSpawnLocation)) ; This is instead of FortAthenaSupplyDrop::SpawnPickup and OnLootSpawned
    (Utilities|Array|RemoveIndex _itemstospawn_2)
    (if (Utilities|Array|IsNotEmpty _itemstospawn_2)
      (Utilities|FlowControl|DelayUntilNextTick)
      (Math|Integer|IncrementInt (Variables|Default|GetItemSpawnCount)
        (:%20%20
          (CallFunction|DelaySpawnLootfromItemComponent))))))

(event EventOnDeathServer
  (params (Damage :type Float%20%28single-precision%29) (DamageTags :type Gameplay%20Tag%20Container%20Structure%20%28by%20ref%29) (Momentum :type Vector :default "0, 0, 0") (HitInfo :type Hit%20Result%20Structure%20%28by%20ref%29) (InstigatedBy :type Controller%20Object%20Reference) (DamageCauser :type Actor%20Object%20Reference) (EffectContext :type Gameplay%20Effect%20Context%20Handle%20Structure))
  (Utilities|FlowControl|DoOnce false
    (:Completed
      (Collision|SetActorEnableCollision false)
      (Actor|SetLifeSpan 2.0)
      (Class|DeviceSupplyDropSpawnerV2|HandleSupplyDropDestroyed (Variables|Default|GetParentDevice) self InstigatedBy)
      (if (Variables|Default|GetSpawnLootOnDeath)
        (CallFunction|OpenSupplyDrop)))))

(event EventOnLanding
  (params (ImpactResult :type Hit%20Result%20Structure))
  (bind _parentdevice (Variables|Default|GetParentDevice))
  (|CallParentFunction ImpactResult)
  (if (not (Variables|Default|GetLandedOnce))
    (Variables|Default|SetLandedOnce true)
    (Utilities|IsValid _parentdevice
      (:Is%20Valid
        (GameplayEvent|BroadcastEvent (Variables|/CRD_SupplyDropSpawner/Actors/Device_SupplyDropSpawner_V2.Device_SupplyDropSpawner_V2_C|GetOn%20Landing _parentdevice) _parentdevice (Utilities|Struct|MakeGameplayEventSupplyDropSpawnerTransmitonLand 0))))))

(event MinigameStartedDelegate(FortMinigameProgress)
  (Utilities|IsValid (Variables|Default|GetParentDevice)
    (:Is%20Not%20Valid
      (Variables|Default|SetSpawnLootOnDeath false)
      (Actor|DestroyActor))))

(event MinigameEndedDelegate(FortMinigameProgress)
  (Utilities|IsValid (Variables|Default|GetParentDevice)
    (:Is%20Not%20Valid
      (Variables|Default|SetSpawnLootOnDeath false)
      (Actor|DestroyActor))))

(event EventOnDamageServer
  (params (Damage :type Float%20%28single-precision%29) (DamageTags :type Gameplay%20Tag%20Container%20Structure%20%28by%20ref%29) (Momentum :type Vector :default "0, 0, 0") (HitInfo :type Hit%20Result%20Structure%20%28by%20ref%29) (InstigatedBy :type Controller%20Object%20Reference) (DamageCauser :type Actor%20Object%20Reference) (EffectContext :type Gameplay%20Effect%20Context%20Handle%20Structure))
  (bind _controller (CreativeAI|Controller|GetParticipantControllerfromActor InstigatedBy false))
  ;[1 1 1] Prevent Players who shouldn't be able to interact with this from destroying it
  (Utilities|IsValid _controller
    (:Is%20Valid
      (if (or (not (CreativeUserRestrictions|CanParticipantControllerInteract (Variables|Default|GetCreative_UseRestriction_Damage) _controller false true false)) (Variables|Default|GetbLocked))
        (Health|SetHealth (+ Damage (Health|GetHealth)))
        (CallFunction|ClearPlayedDying)))))

(fn BlueprintCanInteract
  (params (InteractingPawn :type Fort%20Pawn%20Object%20Reference) (InteractionBeingAttempted :type EInteractionBeingAttempted%20Enum :default "FirstInteraction") (InteractionType :type TInteractionType%20Enum :default "IT_NoInteraction"))
  (return (and (|CallParentFunction InteractingPawn InteractionBeingAttempted InteractionType) (and (not (Variables|Default|GetbLocked)) (CreativeUserRestrictions|CanParticipantActorInteract (Variables|Default|GetCreative_UseRestriction_Component) InteractingPawn false true false)))))

(fn MuteAudio
  (params (MuteAudio :type Boolean))
  (if MuteAudio
    (Audio|Components|Audio|SetVolumeMultiplier (Variables|Default|GetReticleSound_Looping) 0.0 :!class /Script/Engine.AudioComponent)))

(fn OnRep_FXColor
  (bind _fxcolor (Variables|Default|GetFXColor))
  (Networking|CanExecuteCosmeticEvents
    (:True
      (if (not (Variables|/Script/Engine.SceneComponent|GetbHiddenInGame (Variables|Default|GetFlare_Niagara)))
        (Effects|Components|ParticleSystem|SetColorParameter (Variables|Default|GetFlare_Niagara) "FXColor" _fxcolor)
        (Effects|Components|ParticleSystem|SetColorParameter (Variables|Default|GetReticle_Niagara) "Rings" _fxcolor)))))

(fn OnRep_SupplyDropSpawnerReplicationData
  (bind (_ballonhealth_59_d201d9ea409d428dcf9fb793ddd4be0f _bdestructibleballoon_5_ec6de7164fe4891f6f812a886c5165cb _modifymaxhealthclass_8_9267812849bafed41d45338dec7361ad _setmaxhealthdatatag_11_fb51d36b4fc4780b1bf997a7195940fc _fallspeed_14_17083faa4e3bb5af1f539a97bccffc27 _startlocked_17_5cdad18241633fc384b06495e1cdd4cf _itemstospawn_58_af1cf0344c51389e47291b926b3c4812 _showflarewherelanding_25_151b3d8e4c220fbe2274a1b40a49a280 _showicononminimap_27_12ebf43d4acaa803646ba8af8299ea62 _balloonmesh_31_4d96f63c40c0c936fb1cfcbde79aa7e3 _cratemesh_33_5cae5b664f2a4bd62d00398fec7a173b _spawnwithoutballoon_36_53abb48447dd0415121f489eec78ee95 _ballonstyle_39_3d850d344c919ec561d32b9c60f8386d _owningteam_46_014c1604484bcc6f29a968b9c413445c _owningclass_47_6647b8064362a113893a9a97a0d97ff2 _invertteamselection_52_3aaa82324a91316f9cfd6686b80aeaf2 _invertclassselection_53_3a7e099245e1b8a9cfa1279aa14d099c _invertteamdamage_56_0e4c75564ea7325df8603698205ae789 _invertclassdamage_57_696644e645c78348bbf88096fd6b5f98) (Utilities|Struct|BreakSupplyDropSpawnerReplicationData (Variables|Default|GetSupplyDropSpawnerReplicationData)))
  (bind (_fallspeed_14_17083faa4e3bb5af1f539a97bccffc27_2 _startlocked_17_5cdad18241633fc384b06495e1cdd4cf_2 _itemstospawn_58_af1cf0344c51389e47291b926b3c4812_2 _showflarewherelanding_25_151b3d8e4c220fbe2274a1b40a49a280_2 _showicononminimap_27_12ebf43d4acaa803646ba8af8299ea62_2) (Utilities|Struct|BreakSupplyDropSpawnerReplicationData (Variables|Default|GetSupplyDropSpawnerReplicationData)))
  (bind (_ballonhealth_59_d201d9ea409d428dcf9fb793ddd4be0f_2 _bdestructibleballoon_5_ec6de7164fe4891f6f812a886c5165cb_2 _modifymaxhealthclass_8_9267812849bafed41d45338dec7361ad_2 _setmaxhealthdatatag_11_fb51d36b4fc4780b1bf997a7195940fc_2 _fallspeed_14_17083faa4e3bb5af1f539a97bccffc27_3 _startlocked_17_5cdad18241633fc384b06495e1cdd4cf_3 _itemstospawn_58_af1cf0344c51389e47291b926b3c4812_3 _showflarewherelanding_25_151b3d8e4c220fbe2274a1b40a49a280_3 _showicononminimap_27_12ebf43d4acaa803646ba8af8299ea62_3 _balloonmesh_31_4d96f63c40c0c936fb1cfcbde79aa7e3_2 _cratemesh_33_5cae5b664f2a4bd62d00398fec7a173b_2 _spawnwithoutballoon_36_53abb48447dd0415121f489eec78ee95_2 _ballonstyle_39_3d850d344c919ec561d32b9c60f8386d_2 _owningteam_46_014c1604484bcc6f29a968b9c413445c_2 _owningclass_47_6647b8064362a113893a9a97a0d97ff2_2 _invertteamselection_52_3aaa82324a91316f9cfd6686b80aeaf2_2 _invertclassselection_53_3a7e099245e1b8a9cfa1279aa14d099c_2 _invertteamdamage_56_0e4c75564ea7325df8603698205ae789_2 _invertclassdamage_57_696644e645c78348bbf88096fd6b5f98_2 _instigatorteam_62_d0420e7c49553dc95a16be82f5995c39 _instigatorisvalid_65_16ac469c4561da213430cfbd067742da) (Utilities|Struct|BreakSupplyDropSpawnerReplicationData (Variables|Default|GetSupplyDropSpawnerReplicationData)))
  (bind _returnvalue (not _showflarewherelanding_25_151b3d8e4c220fbe2274a1b40a49a280_2))
  (bind _supplydropspawnerreplicationdata (Variables|Default|GetSupplyDropSpawnerReplicationData))
  (bind _fortminigameprogress (Variables|Default|GetFortMinigameProgress))
  (Class|FortMinigameLogicComponent|TryAttachtoMinigame _fortminigameprogress)
  (Initialization|Initialize (Variables|Default|GetCreative_UseRestriction_Damage) _fortminigameprogress 0 0 _owningteam_46_014c1604484bcc6f29a968b9c413445c _invertteamdamage_56_0e4c75564ea7325df8603698205ae789 _owningclass_47_6647b8064362a113893a9a97a0d97ff2 _invertclassdamage_57_696644e645c78348bbf88096fd6b5f98 false)
  (Initialization|Initialize (Variables|Default|GetCreative_UseRestriction_Component) (Variables|Default|GetFortMinigameProgress) 0 0 _owningteam_46_014c1604484bcc6f29a968b9c413445c _invertteamselection_52_3aaa82324a91316f9cfd6686b80aeaf2 _owningclass_47_6647b8064362a113893a9a97a0d97ff2 _invertclassselection_53_3a7e099245e1b8a9cfa1279aa14d099c false)
  (CallFunction|UpdateUseRestrictionwithInstigator _owningteam_46_014c1604484bcc6f29a968b9c413445c_2 _invertteamselection_52_3aaa82324a91316f9cfd6686b80aeaf2_2 _instigatorteam_62_d0420e7c49553dc95a16be82f5995c39 _instigatorisvalid_65_16ac469c4561da213430cfbd067742da)
  (Networking|SwitchHasAuthority
    (:Authority
      (Variables|Default|SetbLocked _startlocked_17_5cdad18241633fc384b06495e1cdd4cf_2)
      (Default|SetSupplyDrop _itemstospawn_58_af1cf0344c51389e47291b926b3c4812_2))
    (:Remote
      _))
  (Development|SetHiddeninGame (Variables|Default|GetReticle_Niagara) _returnvalue false)
  (Default|MuteAudio _returnvalue)
  (Components|Activation|SetActive (Variables|Default|GetAthenaSpecialActor) _showicononminimap_27_12ebf43d4acaa803646ba8af8299ea62_2 false)
  (CallFunction|SetFallSpeed _fallspeed_14_17083faa4e3bb5af1f539a97bccffc27_2))

(fn OpenSupplyDrop
  (if (not (Variables|Runtime|GetbOpened))
    (Class|DeviceSupplyDropSpawnerV2|ClearSpawnedSupplyDrop (Variables|Default|GetParentDevice))))

(fn SetFallSpeed
  (params (SpawnerFallSpeed :type Float%20%28double-precision%29))
  (bind _fortprojectilemovement (Variables|Default|GetFortProjectileMovement))
  (bind _output_get (Variables|/Script/Engine.ProjectileMovementComponent|SetInitialSpeed (* (ScalableFloat|GetValueatLevel (Variables|Hotfix|GetHotfix_InitialFallSpeed)) (Variables|GetSpawnerFallSpeed)) _fortprojectilemovement))
  (Variables|Hotfix|SetHotfix_InitialFallSpeed (Utilities|Struct|MakeScalableFloat _output_get))
  (bind _output_get_2 (Variables|/Script/Engine.ProjectileMovementComponent|SetMaxSpeed (* (ScalableFloat|GetValueatLevel (Variables|Hotfix|GetHotfix_MaxFallSpeed)) (Variables|GetSpawnerFallSpeed)) _fortprojectilemovement))
  (Variables|Hotfix|SetHotfix_MaxFallSpeed (Utilities|Struct|MakeScalableFloat _output_get_2)))

(fn SetSupplyDrop
  (params (ItemsToSpawn :type Array%20of%20Fort%20World%20Item%20Definition%20Object%20References))
  (Networking|SwitchHasAuthority
    (:Authority
      (if (Utilities|Array|IsNotEmpty ItemsToSpawn)
        (Variables|Default|SetItemsToSpawn ItemsToSpawn)))))

(fn ShouldDie
  (params (Damage :type Float%20%28single-precision%29) (EventInstigator :type Controller%20Object%20Reference) (DamageCauser :type Actor%20Object%20Reference) (bIsTest :type Boolean) (EffectContext :type Gameplay%20Effect%20Context%20Handle%20Structure%20%28by%20ref%29))
  (bind _controller (CreativeAI|Controller|GetParticipantControllerfromActor EventInstigator false))
  (Utilities|IsValid EventInstigator
    (:Is%20Valid
      (Utilities|IsValid _controller
        (:Is%20Valid
          (return (CreativeUserRestrictions|CanParticipantControllerInteract (Variables|Default|GetCreative_UseRestriction_Damage) _controller false true false)))
        (:Is%20Not%20Valid
          _)))
    (:Is%20Not%20Valid
      _))
  (return true))

(fn SpawnLoot
  (if (Utilities|Array|IsEmpty (Variables|Default|GetItemsToSpawn))
    (|CallParentFunction)
    (else
      (CallFunction|DelaySpawnLootfromItemComponent))))

(fn UpdateUseRestrictionWithInstigator
  (params (OwningTeam :type Creative%20Team%20Option%20Structure) (InvertTeamSelection :type Boolean) (InstigatorTeam :type Byte) (InstigatorIsValid :type Boolean))
  ;[0 1 0.09829 0.5] If the instigator should be the owner, call SetTeamsWithAffects with the instigator's team. This makes it so only members of the instigator's team can open supply drops.
  (if (Creative|TeamOption|IsTeamType OwningTeam "ActivatingTeam")
    (if (Variables|GetInstigatorIsValid)
      (TeamRestrictions|SetTeamswithAffects (Variables|Default|GetCreative_UseRestriction_Component) (Knobs|Team|ReshiftTeamIndexValue (Math|Conversions|ToInteger(Byte) (Variables|GetInstigatorTeam))) (Variables|GetInvertTeamSelection))
      (else
        (Development|PrintString "CRD_SupplyDropSpawner_V2::UpdateUseRestrictionWithInstigator Warning: Instigator is NOT valid." true true "(R=1.000000,G=0.788867,B=0.000000,A=1.000000)" 1e+01)))))

(fn ConstructionScript
  (|CallParentFunction))
