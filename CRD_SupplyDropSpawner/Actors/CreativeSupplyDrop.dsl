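; Creative supply drop crate, owned by a supply drop spawner device, that falls under a balloon and yields loot.
; BeginPlay lifts it above ground, places the reticle, then offsets it for angled descent.
; Landing, balloon pops and opening broadcast device events; use restrictions gate interaction and death, refunding disallowed damage.
; Loot comes from ItemsToSpawn, one pickup per zero-length delay, or a loot table when that list is empty.
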(event EventBeginPlay
  (bind _childactor (Variables|Default|GetChildActor))
  (bind _childactor_2 (Variables|/Script/Engine.ChildActorComponent|GetChildActor _childactor))
  (bind _reticle_niagara (Variables|Default|GetReticle_Niagara))
  ;[1 1 1] BeginPlay - Initialize fall, balloon, and timers / callbacks / set up location and direction
  (Utilities|FlowControl|Sequence
    (:then_0
      (Networking|SwitchHasAuthority
        (:Authority
          (Utilities|Time|SetTimerbyEvent (EventDispatchers|CreateEvent :!event Multicast_ApplyGravityForFall) (Variables|Default|GetInitialFallDelay)))))
    (:then_1
      (CallFunction|InitFallHeightandSpeed)
      (Utilities|IsValid _childactor_2
        (:Is%20Valid
          (bind _ascreative_supply_drop_balloon (Utilities|Casting|CastToCreativeSupplyDropBalloon _childactor_2))
          (Transformation|AttachComponentToComponent (Variables|/Script/Engine.Actor|GetRootComponent _childactor_2) (Variables|Default|GetCrateStaticMesh) "None" "KeepWorld" "SnapToTarget" "KeepWorld" true)
          (Default|BindEventtoOnBalloonDestroyed _ascreative_supply_drop_balloon (EventDispatchers|CreateEvent :!event OnCreativeBalloonDestroyed)))))
    (:then_2
      ;;[1 1 1] Set up initial position directly above the ground
      (Networking|SwitchHasAuthority
        (:Authority
          (Transformation|Teleport (Math|Vector|MakeVector (.x (Transformation|GetActorLocation)) (.y (Transformation|GetActorLocation)) (+ (.z (Transformation|GetActorLocation)) (Variables|Default|GetFallHeight))) (Math|Rotator|SelectRotator (Variables|Default|GetOverrideRotation) (Transformation|GetActorRotation) (Variables|Default|GetbOverrideRotation))))))
    (:then_3
      ;;[1 1 1] Made direction straight down for placing the reticle. We purposely allow non cosmetic in reticle location so we can use that location for moving the drop along its offset later
      (Variables|Default|SetFallDirection "0, 0,-1.000000")
      (CallFunction|UpdateReticleLocation)
      (Networking|CanExecuteCosmeticEvents
        (:True
          (Audio|PlaySoundatLocation (Variables|Sound|GetSound_Appear) (Transformation|GetActorLocation))
          (Audio|PlaySound2D (Variables|Sound|GetSound%20Appear%20Dialogue))
          (CallFunction|TimerCheckPlayerDistance)
          (if (Utilities|IsValid _reticle_niagara :!class /Script/Engine.KismetSystemLibrary)
            (bind _sound_reticleappear (Variables|Sound|GetSound_ReticleAppear)
              (:then
                (Audio|PlaySoundatLocation _sound_reticleappear (Transformation|GetWorldLocation _reticle_niagara))
                (Utilities|Time|SetTimerbyEvent (EventDispatchers|CreateEvent :!event UpdateReticleLocation) (Variables|Default|GetReticleUpdateInterval) true)
                (Audio|Components|Audio|FadeIn (Variables|Default|GetFlareAudio) 3 1.0 0.0 "Linear" :!class /Script/Engine.AudioComponent))
              (:else))))))
    (:then_4
      ;;[1 1 1] Add whatever offset the drop wants (for coming in at an angle) and resetting the direction
      (Transformation|Teleport (+ (Transformation|GetActorLocation) (Math|Vector|MakeVector (Math|Random|RandomFloatinRange (Variables|Default|GetDescentOffsetMin) (Variables|Default|GetDescentOffsetMax)) (Math|Random|RandomFloatinRange (Variables|Default|GetDescentYOffsetMin) (Variables|Default|GetDescentYOffsetMax)))) (Transformation|GetActorRotation))
      (Variables|Default|SetFallDirection (Math|Vector|Normalize (- (Variables|Default|GetGroundLocation) (Transformation|GetActorLocation)))))
    (:then_5
      ;;[0 1 0.2552 0.5] Update the spawner's use restriction with the instigator on clients and server, and cache the spawner's use restriction component.
      (bind _ascrd_supply_drop_spawner (Utilities|Casting|CastToCRD_SupplyDropSpawner (Actor|GetOwner))
        (:then
          (Class|CRDSupplyDropSpawner|UpdateUseRestrictionwithInstigator _ascrd_supply_drop_spawner (Game|GetInstigator))
          (Variables|Default|SetDamageRestriction (Variables|/CRD_SupplyDropSpawner/Actors/CRD_SupplyDropSpawner.CRD_SupplyDropSpawner_C|GetCreative_UseRestriction_Component _ascrd_supply_drop_spawner)))
        (:CastFailed
          (Development|PrintString "CreativeSupplyDrop::UpdateUseRestrictionWithInstigator Error: Cast to CRD_SupplyDropSpawner failed." true true "(R=1.000000,G=0.788867,B=0.000000,A=1.000000)" 1e+01))))))

(event Interaction|EventBlueprintOnInteract
  (params (InteractingPawn :type Fort%20Pawn%20Object%20Reference) (InteractionBeingAttempted :type EInteractionBeingAttempted%20Enum :default "FirstInteraction"))
  ;[1 1 1] On Interact
  (Utilities|FlowControl|Sequence
    (:then_0
      (bind _output_get (Variables|Default|SetMyInteractingPawn InteractingPawn))
      (CallFunction|OpenSupplyDrop)
      (Utilities|IsValid (Variables|Default|GetParentDevice)
        (:Is%20Valid
          (GameplayEvent|BroadcastEvent (Variables|/CRD_SupplyDropSpawner/Actors/CRD_SupplyDropSpawner.CRD_SupplyDropSpawner_C|GetOn%20Opened (Variables|Default|GetParentDevice)) (Variables|Default|GetParentDevice) (Utilities|Struct|MakeGameplayEventSupplyDropSpawnerTransmitonOpen (Pawn|GetController _output_get))))))
    (:then_1
      (Variables|Default|SetMyInteractingPawn 0))))

(event Custom|DelaySpawnLootFromItemComponent
  (bind (_returnvalue_pickuptospawn _returnvalue_pickupquantity) (Utilities|Struct|BreakMinigameSpawnerSpawnParams (ItemData|MakeSpawnParams (Variables|/CRD_SupplyDropSpawner/Actors/CRD_SupplyDropSpawner.CRD_SupplyDropSpawner_C|GetPickupItemListComponent (Variables|Default|GetParentDevice)) (Variables|Default|GetItemSpawnCount) (Variables|Default|GetSpawn%20Transform))))
  (bind _itemstospawn (Variables|Default|GetItemsToSpawn))
  (bind _itemstospawn_2 (Variables|Default|GetItemsToSpawn))
  ;[1 1 1] Spawn Loot from ItemsToSpawn
  (if (Utilities|Array|IsNotEmpty _itemstospawn)
    (CallFunction|SpawnPickup (Utilities|Array|Get(acopy) _itemstospawn) _returnvalue_pickupquantity (Variables|Default|GetMyInteractingPawn) (CallFunction|GetLootSpawnLocation))
    (CallFunction|OnLootSpawned)
    (Utilities|Array|RemoveIndex _itemstospawn_2)
    (if (Utilities|Array|IsNotEmpty _itemstospawn_2)
      (Utilities|FlowControl|Delay 0.0)
      (Math|Integer|IncrementInt (Variables|Default|GetItemSpawnCount)
        (:%20%20
          (CallFunction|DelaySpawnLootfromItemComponent))))))

(event EventOnDeathServer
  (params (Damage :type Float%20%28single-precision%29) (DamageTags :type Gameplay%20Tag%20Container%20Structure%20%28by%20ref%29) (Momentum :type Vector :default "0, 0, 0") (HitInfo :type Hit%20Result%20Structure%20%28by%20ref%29) (InstigatedBy :type Controller%20Object%20Reference) (DamageCauser :type Actor%20Object%20Reference) (EffectContext :type Gameplay%20Effect%20Context%20Handle%20Structure))
  ;[1 1 1] OnDeathServer
  (Utilities|FlowControl|DoOnce false
    (:Completed
      (Collision|SetActorEnableCollision false)
      (Actor|SetLifeSpan 2.0)
      (if (Variables|Default|GetSpawnLootOnDeath)
        (CallFunction|OpenSupplyDrop)))))

(event OnProjectileStop(FortProjectileMovement)
  (params (ImpactResult :type Hit%20Result%20Structure%20%28by%20ref%29))
  (bind _parentdevice (Variables|Default|GetParentDevice))
  (CallFunction|HandleProjectileMovementStop ImpactResult)
  (CallFunction|OnLanding)
  (CallFunction|DestroyGroundComponents)
  (if (not (Variables|Default|GetLandedOnce))
    (Variables|Default|SetLandedOnce true)
    (Utilities|IsValid _parentdevice
      (:Is%20Valid
        (GameplayEvent|BroadcastEvent (Variables|/CRD_SupplyDropSpawner/Actors/CRD_SupplyDropSpawner.CRD_SupplyDropSpawner_C|GetOn%20Landing _parentdevice) _parentdevice (Utilities|Struct|MakeGameplayEventSupplyDropSpawnerTransmitonLand 0))))))

(event MinigameStartedDelegate(FortMinigameProgress)
  (Utilities|IsValid (Variables|Default|GetParentDevice)
    (:Is%20Not%20Valid
      (Variables|Default|SetSpawnLootOnDeath false)
      (Actor|DestroyActor))))

(event MinigameEndedDelegate(FortMinigameProgress)
  (Utilities|IsValid (Variables|Default|GetParentDevice)
    (:Is%20Not%20Valid
      (Variables|Default|SetSpawnLootOnDeath false)
      (Actor|DestroyActor))))

(event EventOnDamageServer
  (params (Damage :type Float%20%28single-precision%29) (DamageTags :type Gameplay%20Tag%20Container%20Structure%20%28by%20ref%29) (Momentum :type Vector :default "0, 0, 0") (HitInfo :type Hit%20Result%20Structure%20%28by%20ref%29) (InstigatedBy :type Controller%20Object%20Reference) (DamageCauser :type Actor%20Object%20Reference) (EffectContext :type Gameplay%20Effect%20Context%20Handle%20Structure))
  ;[1 1 1] Prevent Players who shouldn't be able to interact with this from destroying it
  (bind _asplayer_controller (Utilities|Casting|CastToPlayerController InstigatedBy))
  (bind _damagerestriction (Variables|Default|GetDamageRestriction)
    (:then
      (if (not (DEPRECATED|CanPlayerControllerInteract _damagerestriction _asplayer_controller true false))
        (Health|SetHealth (+ Damage (Health|GetHealth)))
        (CallFunction|ClearPlayedDying)))
    (:else)))

(fn BalloonDestroyed
  (|CallParentFunction))

(fn BlueprintCanInteract
  (params (InteractingPawn :type Fort%20Pawn%20Object%20Reference) (InteractionBeingAttempted :type EInteractionBeingAttempted%20Enum :default "FirstInteraction") (InteractionType :type TInteractionType%20Enum :default "IT_NoInteraction"))
  (bind _restrictioncomponent (Variables|Default|GetRestrictionComponent)
    (:then
      (return (Math|Boolean|ANDBoolean (not (Variables|Default|GetbLocked)) (DEPRECATED|CanActorInteract _restrictioncomponent InteractingPawn true false) (not (Variables|Default|GetOpened)))))
    (:else
      (return true))))

(fn InitFallHeightAndSpeed
  (Variables|Default|SetFallTime_0 (/ (Variables|Default|GetFallHeight) (Variables|Default|GetFallSpeed)))
  (Networking|CanExecuteCosmeticEvents
    (:True
      (Utilities|Time|SetTimerbyEvent (EventDispatchers|CreateEvent :!event DisableFlare) (* (Variables|Default|GetFallTime_0) 0.7))
      (Utilities|Time|SetTimerbyEvent (EventDispatchers|CreateEvent :!event AlmostLanded) (* (Variables|Default|GetFallTime_0) 0.95)))))

(fn OnCreativeBalloonDestroyed
  (params (Controller :type Controller%20Object%20Reference))
  (bind _parentdevice (Variables|Default|GetParentDevice))
  ;[1 1 1] If the supply drop fall speed is set to 0, the balloon is popped as soon as it spawns.
  (Utilities|IsValid _parentdevice
    (:Is%20Valid
      (if (not (== (Variables|/CRD_SupplyDropSpawner/Actors/CRD_SupplyDropSpawner.CRD_SupplyDropSpawner_C|GetfFallSpeedMultiplier _parentdevice) 0))
        (GameplayEvent|BroadcastEvent (Variables|/CRD_SupplyDropSpawner/Actors/CRD_SupplyDropSpawner.CRD_SupplyDropSpawner_C|GetOn%20Balloon%20Popped _parentdevice) _parentdevice (Utilities|Struct|MakeGameplayEventSupplyDropSpawnerTransmitonPop Controller)))
      (CallFunction|BalloonDestroyed))
    (:Is%20Not%20Valid
      _)))

(fn OnRep_FXColor
  (bind _fxcolor (Variables|Default|GetFXColor))
  (Networking|CanExecuteCosmeticEvents
    (:True
      (if (not (Variables|/Script/Engine.SceneComponent|GetbHiddenInGame (Variables|Default|GetFlare_Niagara)))
        (Effects|Components|ParticleSystem|SetColorParameter (Variables|Default|GetFlare_Niagara) "FXColor" _fxcolor)
        (Effects|Components|ParticleSystem|SetColorParameter (Variables|Default|GetReticle_Niagara) "Rings" _fxcolor)))))

(fn OpenSupplyDrop
  (if (not (Variables|Default|GetOpened))
    (Class|CRDSupplyDropSpawner|ClearSpawnedSupplyDrop (Variables|Default|GetParentDevice))
    (|CallParentFunction)))

(fn SetAudioLevel
  (params (MuteAudio :type Boolean))
  (if MuteAudio
    (Audio|Components|Audio|SetVolumeMultiplier (Variables|Default|GetFlareAudio) 0.0 :!class /Script/Engine.AudioComponent)))

(fn SetFlareColor
  (params (FlareColor :type Linear%20Color%20Structure))
  (Variables|Default|SetFXColor FlareColor))

(fn SetParentDevice
  (params (ParentDevice :type CRD%20Supply%20Drop%20Spawner%20Object%20Reference))
  (Variables|Default|SetParentDevice ParentDevice))

(fn ShouldDie
  (params (Damage :type Float%20%28single-precision%29) (EventInstigator :type Controller%20Object%20Reference) (DamageCauser :type Actor%20Object%20Reference) (bIsTest :type Boolean) (EffectContext :type Gameplay%20Effect%20Context%20Handle%20Structure%20%28by%20ref%29))
  (Utilities|IsValid EventInstigator
    (:Is%20Valid
      (bind _asplayer_controller (Utilities|Casting|CastToPlayerController EventInstigator)
        (:then
          (bind _damagerestriction (Variables|Default|GetDamageRestriction)
            (:then
              (return (DEPRECATED|CanPlayerControllerInteract _damagerestriction _asplayer_controller true false)))
            (:else
              _)))
        (:CastFailed
          _))
      (return true))
    (:Is%20Not%20Valid
      _)))

(fn SpawnLoot
  (if (Utilities|Array|IsEmpty (Variables|Default|GetItemsToSpawn))
    (bind _output_get (Variables|Default|SetLootTableNameToUse (Variables|Default|GetLootTableName)))
    (bind (_outloottodrop _returnvalue) (Loot|PickLootDrops _output_get (Loot|GetLootLevel) -1 "(GameplayTags=)"))
    (if _returnvalue
      (Variables|Default|SetDelayedSpawnLoot _outloottodrop)
      (CallFunction|OnDelayedSpawnLoot))
    (else
      (CallFunction|DelaySpawnLootfromItemComponent))))

(fn ConstructionScript
  (|CallParentFunction))
