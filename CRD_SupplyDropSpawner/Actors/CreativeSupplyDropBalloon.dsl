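; Destructible balloon actor for the Creative supply drop spawner that applies a damage-reduction effect to itself and notifies the drop when destroyed.
; A damage restriction decides which participants count, gating death, hit effects and damage numbers.
; Server damage from disallowed participants is undone by adding it back to health, while the first allowed hit reveals the health bar.
; Death hides the balloon, plays pop effects at its "boom" socket and expires it shortly after.
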(event EventBeginPlay
  ;[1 1 1] hide health bar on beginplay
  (HealthBarComponent|SetHealthBarVisibility (Variables|Default|GetFortHealthBar) false)
  (Environment|ApplyGlobalEnvironmentGameplayEffecttoActor self (Variables|Default|GetGE_DamageReduction)))

(event Collision|EventActorBeginOverlap
  (params (OtherActor :type Actor%20Object%20Reference)))

(event EventOnDamageServer
  (params (Damage :type Float%20%28single-precision%29) (DamageTags :type Gameplay%20Tag%20Container%20Structure%20%28by%20ref%29) (Momentum :type Vector :default "0, 0, 0") (HitInfo :type Hit%20Result%20Structure%20%28by%20ref%29) (InstigatedBy :type Controller%20Object%20Reference) (DamageCauser :type Actor%20Object%20Reference) (EffectContext :type Gameplay%20Effect%20Context%20Handle%20Structure))
  (bind _controller (CreativeAI|Controller|GetParticipantControllerfromActor InstigatedBy false))
  ;[1 1 1] handle damage
  (Utilities|IsValid _controller
    (:Is%20Valid
      (bind _damagerestriction (Variables|Default|GetDamageRestriction)
        (:then
          (if (CreativeUserRestrictions|CanParticipantControllerInteract _damagerestriction _controller false true false)
            ;;[1 1 1] make hp bar visible after taking damage
            (Utilities|FlowControl|DoOnce false
              (:Completed
                (HealthBarComponent|SetHealthBarVisibility (Variables|Default|GetFortHealthBar) true)))
            (else
              (Health|SetHealth (+ Damage (Health|GetHealth))))))
        (:else)))))

(event EventTick
  (params (DeltaSeconds :type Float%20%28single-precision%29)))

(event EventOnDeathServer
  (params (Damage :type Float%20%28single-precision%29) (DamageTags :type Gameplay%20Tag%20Container%20Structure%20%28by%20ref%29) (Momentum :type Vector :default "0, 0, 0") (HitInfo :type Hit%20Result%20Structure%20%28by%20ref%29) (InstigatedBy :type Controller%20Object%20Reference) (DamageCauser :type Actor%20Object%20Reference) (EffectContext :type Gameplay%20Effect%20Context%20Handle%20Structure))
  ;[1 1 1] OnDeathServer
  (Variables|Actor|SetbCanBeDamaged false)
  (Default|CallOnBalloonDestroyed InstigatedBy) ; Let supply drop know
  (Actor|SetLifeSpan 2.0))

(event EventOnDeathPlayEffects
  (params (Damage :type Float%20%28single-precision%29) (DamageTags :type Gameplay%20Tag%20Container%20Structure%20%28by%20ref%29) (Momentum :type Vector :default "0, 0, 0") (HitInfo :type Hit%20Result%20Structure%20%28by%20ref%29) (InstigatedBy :type Fort%20Pawn%20Object%20Reference) (DamageCauser :type Actor%20Object%20Reference) (EffectContext :type Gameplay%20Effect%20Context%20Handle%20Structure))
  (bind _staticmeshcomponent (Variables|Debug|GetStaticMeshComponent))
  (bind _returnvalue (Transformation|GetSocketLocation _staticmeshcomponent "boom"))
  ;[1 1 1] OnDeathPlayeffects
  (Rendering|SetActorHiddenInGame true)
  (bind _returnvalue_2 (Effects|Components|ParticleSystem|SpawnFXAtLocation (Variables|Default|GetBalloonPopFX) _returnvalue))
  (Components|Activation|SetActive _returnvalue_2 true false)
  (Audio|PlaySoundatLocation (Variables|Default|GetBalloonPopSound) _returnvalue)
  (HealthBarComponent|SetHealthBarVisibility (Variables|Default|GetFortHealthBar) false))

(event EventOnDamagePlayEffects
  (params (Damage :type Float%20%28single-precision%29) (DamageTags :type Gameplay%20Tag%20Container%20Structure%20%28by%20ref%29) (Momentum :type Vector :default "0, 0, 0") (HitInfo :type Hit%20Result%20Structure%20%28by%20ref%29) (InstigatedBy :type Fort%20Pawn%20Object%20Reference) (DamageCauser :type Actor%20Object%20Reference) (EffectContext :type Gameplay%20Effect%20Context%20Handle%20Structure))
  (bind _controller (CreativeAI|Controller|GetParticipantControllerfromActor InstigatedBy false))
  (Utilities|IsValid _controller
    (:Is%20Valid
      (bind _damagerestriction (Variables|Default|GetDamageRestriction)
        (:then
          (if (CreativeUserRestrictions|CanParticipantControllerInteract _damagerestriction _controller true true false)
            (|CallParentFunction Damage DamageTags Momentum HitInfo InstigatedBy DamageCauser EffectContext)))
        (:else)))))

(fn ShouldDie
  (params (Damage :type Float%20%28single-precision%29) (EventInstigator :type Controller%20Object%20Reference) (DamageCauser :type Actor%20Object%20Reference) (bIsTest :type Boolean) (EffectContext :type Gameplay%20Effect%20Context%20Handle%20Structure%20%28by%20ref%29))
  (Utilities|IsValid EventInstigator
    (:Is%20Valid
      (Utilities|IsValid EventInstigator
        (:Is%20Valid
          (bind _damagerestriction (Variables|Default|GetDamageRestriction)
            (:then
              (return (CreativeUserRestrictions|CanParticipantControllerInteract _damagerestriction (CreativeAI|Controller|GetParticipantControllerfromActor EventInstigator false) false true false)))
            (:else)))))
    (:Is%20Not%20Valid
      (return true))))

(fn ShouldDisplayDamageNumbers
  (Networking|CanExecuteCosmeticEvents
    (:True
      (bind _damagerestriction (Variables|Default|GetDamageRestriction)
        (:then
          (return (CreativeUserRestrictions|CanParticipantControllerInteract _damagerestriction (Fortnite|GetFirstLocalFortPlayerController true) false true false)))
        (:else
          _)))
    (:False
      _))
  (return false))

(fn ConstructionScript)
