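; Creative supply drop crate for the Supply Drop Spawner V2 device: descends under a child-actor balloon, then dispenses loot when opened.
; BeginPlay raises it above the ground, places the reticle, then offsets it for an angled fall.
; An OnRep of the spawner's data struct applies restrictions, balloon health, fall speed, lock and async-loaded meshes.
; Loot spawns one item per tick or from a loot table; pop, landing and opening broadcast through the device.
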(event EventBeginPlay
  (bind _childactor (Variables|Default|GetChildActor))
  (bind _reticle_niagara (Variables|Default|GetReticle_Niagara))
  (bind _childactor_2 (Variables|/Script/Engine.ChildActorComponent|GetChildActor _childactor))
  ;[1 1 1] BeginPlay - Initialize fall, balloon, and timers / callbacks / set up location and direction
  (Utilities|FlowControl|Sequence
    (:then_0
      (Networking|SwitchHasAuthority
        (:Authority
          (Utilities|Time|SetTimerbyEvent (EventDispatchers|CreateEvent :!event Multicast_ApplyGravityForFall) (Variables|Default|GetInitialFallDelay)))))
    (:then_1
      (CallFunction|InitFallHeightandSpeed)
      (Utilities|IsValid _childactor_2
        (:Is%20Valid
          (bind _ascreative_supply_drop_balloon (Utilities|Casting|CastToCreativeSupplyDropBalloon _childactor_2))
          (bind _output_get (Variables|Default|SetSupplyDropBalloon _ascreative_supply_drop_balloon))
          (Transformation|AttachComponentToComponent (Variables|/Script/Engine.Actor|GetRootComponent _childactor_2) (Variables|Default|GetCrateStaticMesh) "None" "KeepWorld" "SnapToTarget" "KeepWorld" true)
          (Default|BindEventtoOnBalloonDestroyed _output_get (EventDispatchers|CreateEvent :!event OnCreativeBalloonDestroyed)))))
    (:then_2
      ;;[1 1 1] Set up initial position directly above the ground
      (Networking|SwitchHasAuthority
        (:Authority
          (Transformation|Teleport (Math|Vector|MakeVector (.x (Transformation|GetActorLocation)) (.y (Transformation|GetActorLocation)) (+ (.z (Transformation|GetActorLocation)) (Variables|Default|GetFallHeight))) (Math|Rotator|SelectRotator (Variables|Default|GetOverrideRotation) (Transformation|GetActorRotation) (Variables|Default|GetbOverrideRotation))))))
    (:then_3
      ;;[1 1 1] Made direction straight down for placing the reticle. We purposely allow non cosmetic in reticle location so we can use that location for moving the drop along its offset later
      (Variables|Default|SetFallDirection "0, 0,-1.000000")
      (CallFunction|UpdateReticleLocation)
      (Networking|CanExecuteCosmeticEvents
        (:True
          (if (not (Variables|Default|GetbHasValidGroundLocation))
            (Components|DestroyComponent (Variables|Default|GetReticle_Niagara)))
          (Audio|PlaySoundatLocation (Variables|Sound|GetSound_Appear) (Transformation|GetActorLocation))
          (Audio|PlaySound2D (Variables|Sound|GetSound%20Appear%20Dialogue))
          (CallFunction|TimerCheckPlayerDistance)
          (if (Utilities|IsValid _reticle_niagara :!class /Script/Engine.KismetSystemLibrary)
            (bind _sound_reticleappear (Variables|Sound|GetSound_ReticleAppear)
              (:then
                (Audio|PlaySoundatLocation _sound_reticleappear (Transformation|GetWorldLocation _reticle_niagara))
                (Utilities|Time|SetTimerbyEvent (EventDispatchers|CreateEvent :!event UpdateReticleLocation) (Variables|Default|GetReticleUpdateInterval) true)
                (Audio|Components|Audio|FadeIn (Variables|Default|GetFlareAudio) 3 1.0 0.0 "Linear" :!class /Script/Engine.AudioComponent))
              (:else))))))
    (:then_4
      ;;[1 1 1] Add whatever offset the drop wants (for coming in at an angle) and resetting the direction
      (Transformation|Teleport (+ (Transformation|GetActorLocation) (Math|Vector|MakeVector (Math|Random|RandomFloatinRange (Variables|Default|GetDescentOffsetMin) (Variables|Default|GetDescentOffsetMax)) (Math|Random|RandomFloatinRange (Variables|Default|GetDescentYOffsetMin) (Variables|Default|GetDescentYOffsetMax)))) (Transformation|GetActorRotation))
      (Variables|Default|SetFallDirection (Math|Vector|Normalize (- (Variables|Default|GetGroundLocation) (Transformation|GetActorLocation)))))))

(event Interaction|EventBlueprintOnInteract
  (params (InteractingPawn :type Fort%20Pawn%20Object%20Reference) (InteractionBeingAttempted :type EInteractionBeingAttempted%20Enum :default "FirstInteraction"))
  ;[1 1 1] On Interact
  (Utilities|FlowControl|Sequence
    (:then_0
      (bind _output_get (Variables|Default|SetMyInteractingPawn InteractingPawn))
      (CallFunction|OpenSupplyDrop)
      (Utilities|IsValid (Variables|Default|GetParentDevice)
        (:Is%20Valid
          (GameplayEvent|BroadcastEvent (Variables|/CRD_SupplyDropSpawner/Actors/Device_SupplyDropSpawner_V2.Device_SupplyDropSpawner_V2_C|GetOn%20Opened (Variables|Default|GetParentDevice)) (Variables|Default|GetParentDevice) (Utilities|Struct|MakeGameplayEventSupplyDropSpawnerTransmitonOpen (Pawn|GetController _output_get))))))
    (:then_1
      (Variables|Default|SetMyInteractingPawn 0))))

(event Custom|DelaySpawnLootFromItemComponent
  (bind (_returnvalue_pickuptospawn _returnvalue_pickupquantity) (Utilities|Struct|BreakMinigameSpawnerSpawnParams (ItemData|MakeSpawnParams (Variables|/CRD_SupplyDropSpawner/Actors/Device_SupplyDropSpawner_V2.Device_SupplyDropSpawner_V2_C|GetPickupItemListComponent (Variables|Default|GetParentDevice)) (Variables|Default|GetItemSpawnCount) (Variables|Default|GetSpawn%20Transform))))
  (bind _itemstospawn (Variables|Default|GetItemsToSpawn))
  (bind _itemstospawn_2 (Variables|Default|GetItemsToSpawn))
  ;[1 1 1] Spawn Loot from ItemsToSpawn
  (if (Utilities|Array|IsNotEmpty _itemstospawn)
    (CallFunction|SpawnPickup (Utilities|Array|Get(acopy) _itemstospawn) _returnvalue_pickupquantity (Variables|Default|GetMyInteractingPawn) (CallFunction|GetLootSpawnLocation))
    (CallFunction|OnLootSpawned)
    (Utilities|Array|RemoveIndex _itemstospawn_2)
    (if (Utilities|Array|IsNotEmpty _itemstospawn_2)
      (Utilities|FlowControl|DelayUntilNextTick)
      (Math|Integer|IncrementInt (Variables|Default|GetItemSpawnCount)
        (:%20%20
          (CallFunction|DelaySpawnLootfromItemComponent))))))

(event EventOnDeathServer
  (params (Damage :type Float%20%28single-precision%29) (DamageTags :type Gameplay%20Tag%20Container%20Structure%20%28by%20ref%29) (Momentum :type Vector :default "0, 0, 0") (HitInfo :type Hit%20Result%20Structure%20%28by%20ref%29) (InstigatedBy :type Controller%20Object%20Reference) (DamageCauser :type Actor%20Object%20Reference) (EffectContext :type Gameplay%20Effect%20Context%20Handle%20Structure))
  ;[1 1 1] OnDeathServer
  (Utilities|FlowControl|DoOnce false
    (:Completed
      (Collision|SetActorEnableCollision false)
      (Actor|SetLifeSpan 2.0)
      (Class|DeviceSupplyDropSpawnerV2|HandleSupplyDropDestroyed (Variables|Default|GetParentDevice) self InstigatedBy)
      (if (Variables|Default|GetSpawnLootOnDeath)
        (CallFunction|OpenSupplyDrop)))))

(event OnProjectileStop(FortProjectileMovement)
  (params (ImpactResult :type Hit%20Result%20Structure%20%28by%20ref%29))
  (bind _parentdevice (Variables|Default|GetParentDevice))
  (CallFunction|HandleProjectileMovementStop ImpactResult)
  (CallFunction|OnLanding)
  (CallFunction|DestroyGroundComponents)
  (if (not (Variables|Default|GetLandedOnce))
    (Variables|Default|SetLandedOnce true)
    (Utilities|IsValid _parentdevice
      (:Is%20Valid
        (GameplayEvent|BroadcastEvent (Variables|/CRD_SupplyDropSpawner/Actors/Device_SupplyDropSpawner_V2.Device_SupplyDropSpawner_V2_C|GetOn%20Landing _parentdevice) _parentdevice (Utilities|Struct|MakeGameplayEventSupplyDropSpawnerTransmitonLand 0))))))

(event MinigameStartedDelegate(FortMinigameProgress)
  (Utilities|IsValid (Variables|Default|GetParentDevice)
    (:Is%20Not%20Valid
      (Variables|Default|SetSpawnLootOnDeath false)
      (Actor|DestroyActor))))

(event MinigameEndedDelegate(FortMinigameProgress)
  (Utilities|IsValid (Variables|Default|GetParentDevice)
    (:Is%20Not%20Valid
      (Variables|Default|SetSpawnLootOnDeath false)
      (Actor|DestroyActor))))

(event EventOnDamageServer
  (params (Damage :type Float%20%28single-precision%29) (DamageTags :type Gameplay%20Tag%20Container%20Structure%20%28by%20ref%29) (Momentum :type Vector :default "0, 0, 0") (HitInfo :type Hit%20Result%20Structure%20%28by%20ref%29) (InstigatedBy :type Controller%20Object%20Reference) (DamageCauser :type Actor%20Object%20Reference) (EffectContext :type Gameplay%20Effect%20Context%20Handle%20Structure))
  (bind _controller (CreativeAI|Controller|GetParticipantControllerfromActor InstigatedBy false))
  ;[1 1 1] Prevent Players who shouldn't be able to interact with this from destroying it
  (Utilities|IsValid _controller
    (:Is%20Valid
      (if (not (CreativeUserRestrictions|CanParticipantControllerInteract (Variables|Default|GetCreative_UseRestriction_Damage) _controller false true false))
        (Health|SetHealth (+ Damage (Health|GetHealth)))
        (CallFunction|ClearPlayedDying)))))

(event Custom|SetSupplyDropMesh
  (params (SpawnWithoutBalloon :type Boolean) (BalloonMesh :type Static%20Mesh%20Soft%20Object%20Reference) (CrateMesh :type Static%20Mesh%20Soft%20Object%20Reference) (BallonStyle :type CRD_SupplyDropSpawner_BalloonStyle%20Enum :default "NewEnumerator0"))
  (bind _supplydropballoon (Variables|Default|GetSupplyDropBalloon))
  (bind _cratestaticmesh (Variables|Default|GetCrateStaticMesh))
  (bind _supplydropballoon_2 (Variables|Default|GetSupplyDropBalloon))
  (bind _staticmeshcomponent (Variables|/Script/FortniteGame.BuildingSMActor|GetStaticMeshComponent _supplydropballoon_2))
  ;[1 1 1] Cosmetic
  (Networking|CanExecuteCosmeticEvents
    (:True
      (if (not SpawnWithoutBalloon)
        (if (Utilities|IsValidSoftObjectReference BalloonMesh)
          (bind _object (Utilities|AsyncLoadAsset BalloonMesh)
            (:then
              _)
            (:Completed
              (Utilities|IsValid _supplydropballoon_2
                (:Is%20Valid
                  (bind _asstatic_mesh (Utilities|Casting|CastToStaticMesh _object))
                  (Components|StaticMesh|SetStaticMesh _staticmeshcomponent _asstatic_mesh)
                  (Transformation|SetRelativeScale3D _staticmeshcomponent "0.800000,0.800000,0.800000")))))))
      (if (Utilities|IsValidSoftObjectReference CrateMesh)
        (bind _object_2 (Utilities|AsyncLoadAsset CrateMesh)
          (:Completed
            (bind _asstatic_mesh_2 (Utilities|Casting|CastToStaticMesh _object_2))
            (Components|StaticMesh|SetStaticMesh _cratestaticmesh _asstatic_mesh_2)
            ;;[1 1 1] We need to adjust the scale for the birthday version of the supply drop
            (switch Utilities|FlowControl|Switch|SwitchonCRD_SupplyDropSpawner_BalloonStyle BallonStyle
              (:NewEnumerator2
                (Transformation|SetRelativeScale3D _cratestaticmesh "2.000000,2.000000,2.000000")
                (Utilities|IsValid _supplydropballoon
                  (:Is%20Valid
                    (Transformation|SetRelativeScale3D (Variables|/Script/FortniteGame.BuildingSMActor|GetStaticMeshComponent _supplydropballoon) "0.420000,0.420000,0.420000")))))))))))

(event EventSetFlareColor
  (params (FlareColor :type Linear%20Color%20Structure))
  ; BPI_CreativeSupplyDrop Interface
  (Variables|Default|SetFXColor FlareColor))

(event EventSetParentDevice
  (params (ParentDevice :type Device%20Supply%20Drop%20Spawner%20V2%20Object%20Reference))
  ; BPI_CreativeSupplyDrop Interface
  (Variables|Default|SetParentDevice ParentDevice))

(event EventSetReplicationData
  (params (ReplicationData :type Supply%20Drop%20Spawner%20Replication%20Data%20Structure))
  ; BPI_CreativeSupplyDrop Interface
  (Variables|Default|SetSupplyDropSpawnerReplicationData ReplicationData))

(event EventSetLocked
  (params (Locked :type Boolean))
  ; BPI_CreativeSupplyDrop Interface
  (Variables|Default|SetbLocked Locked))

(event EventDestroyBalloon
  ; BPI_CreativeSupplyDrop Interface
  (if (not (Variables|Default|GetBalloonPopped))
    (bind _ascreative_supply_drop_balloon (Utilities|Casting|CastToCreativeSupplyDropBalloon (Variables|/Script/Engine.ChildActorComponent|GetChildActor (Variables|Default|GetChildActor))))
    (Actor|SetLifeSpan _ascreative_supply_drop_balloon 0.1)))

(event EventOpenCreativeSupplyDrop
  (params (Instigator :type Controller%20Object%20Reference))
  ; BPI_CreativeSupplyDrop Interface
  (CallFunction|OpenSupplyDrop))

(event EventDestroyCreativeSupplyDrop
  ; BPI_CreativeSupplyDrop Interface
  (Variables|Default|SetSpawnLootOnDeath false)
  (Actor|DestroyActor)
  (bind _ascreative_supply_drop_balloon (Utilities|Casting|CastToCreativeSupplyDropBalloon (Variables|/Script/Engine.ChildActorComponent|GetChildActor (Variables|Default|GetChildActor))))
  (Actor|DestroyActor _ascreative_supply_drop_balloon) ; Destroy the balloon immediately,\nso its collision doesn't block subsequent drops
)

(fn BalloonDestroyed
  (|CallParentFunction))

(fn BlueprintCanInteract
  (params (InteractingPawn :type Fort%20Pawn%20Object%20Reference) (InteractionBeingAttempted :type EInteractionBeingAttempted%20Enum :default "FirstInteraction") (InteractionType :type TInteractionType%20Enum :default "IT_NoInteraction"))
  (return (Math|Boolean|ANDBoolean (not (Variables|Default|GetbLocked)) (CreativeUserRestrictions|CanParticipantActorInteract (Variables|Default|GetCreative_UseRestriction_Component) InteractingPawn false true false) (not (Variables|Default|GetOpened)))))

(fn InitFallHeightAndSpeed
  (Variables|Default|SetFallTime_0 (/ (Variables|Default|GetFallHeight) (Variables|Default|GetFallSpeed)))
  (Networking|CanExecuteCosmeticEvents
    (:True
      (Utilities|Time|SetTimerbyEvent (EventDispatchers|CreateEvent :!event DisableFlare) (* (Variables|Default|GetFallTime_0) 0.7))
      (Utilities|Time|SetTimerbyEvent (EventDispatchers|CreateEvent :!event AlmostLanded) (* (Variables|Default|GetFallTime_0) 0.95)))))

(fn MuteAudio
  (params (MuteAudio :type Boolean))
  (if MuteAudio
    (Audio|Components|Audio|SetVolumeMultiplier (Variables|Default|GetFlareAudio) 0.0 :!class /Script/Engine.AudioComponent)))

(fn OnCreativeBalloonDestroyed
  (params (Controller :type Controller%20Object%20Reference))
  (bind _parentdevice (Variables|Default|GetParentDevice))
  ;[1 1 1] If the supply drop fall speed is set to 0, the balloon is popped as soon as it spawns.
  (Utilities|IsValid _parentdevice
    (:Is%20Valid
      (if (not (== (Variables|/CRD_SupplyDropSpawner/Actors/Device_SupplyDropSpawner_V2.Device_SupplyDropSpawner_V2_C|GetFallSpeed _parentdevice) 0))
        (GameplayEvent|BroadcastEvent (Variables|/CRD_SupplyDropSpawner/Actors/Device_SupplyDropSpawner_V2.Device_SupplyDropSpawner_V2_C|GetOn%20Balloon%20Popped _parentdevice) _parentdevice (Utilities|Struct|MakeGameplayEventSupplyDropSpawnerTransmitonPop Controller)))
      (CallFunction|BalloonDestroyed))
    (:Is%20Not%20Valid
      _)))

(fn OnRep_FXColor
  (bind _fxcolor (Variables|Default|GetFXColor))
  (Networking|CanExecuteCosmeticEvents
    (:True
      (if (not (Variables|/Script/Engine.SceneComponent|GetbHiddenInGame (Variables|Default|GetFlare_Niagara)))
        (Effects|Components|ParticleSystem|SetColorParameter (Variables|Default|GetFlare_Niagara) "FXColor" _fxcolor)
        (Effects|Components|ParticleSystem|SetColorParameter (Variables|Default|GetReticle_Niagara) "Rings" _fxcolor)))))

(fn OnRep_SupplyDropSpawnerReplicationData
  (bind (_ballonhealth_59_d201d9ea409d428dcf9fb793ddd4be0f _bdestructibleballoon_5_ec6de7164fe4891f6f812a886c5165cb _modifymaxhealthclass_8_9267812849bafed41d45338dec7361ad _setmaxhealthdatatag_11_fb51d36b4fc4780b1bf997a7195940fc _fallspeed_14_17083faa4e3bb5af1f539a97bccffc27 _startlocked_17_5cdad18241633fc384b06495e1cdd4cf _itemstospawn_58_af1cf0344c51389e47291b926b3c4812 _showflarewherelanding_25_151b3d8e4c220fbe2274a1b40a49a280 _showicononminimap_27_12ebf43d4acaa803646ba8af8299ea62 _balloonmesh_31_4d96f63c40c0c936fb1cfcbde79aa7e3 _cratemesh_33_5cae5b664f2a4bd62d00398fec7a173b _spawnwithoutballoon_36_53abb48447dd0415121f489eec78ee95 _ballonstyle_39_3d850d344c919ec561d32b9c60f8386d _owningteam_46_014c1604484bcc6f29a968b9c413445c _owningclass_47_6647b8064362a113893a9a97a0d97ff2 _invertteamselection_52_3aaa82324a91316f9cfd6686b80aeaf2 _invertclassselection_53_3a7e099245e1b8a9cfa1279aa14d099c _invertteamdamage_56_0e4c75564ea7325df8603698205ae789 _invertclassdamage_57_696644e645c78348bbf88096fd6b5f98) (Utilities|Struct|BreakSupplyDropSpawnerReplicationData (Variables|Default|GetSupplyDropSpawnerReplicationData)))
  (bind (_ballonhealth_59_d201d9ea409d428dcf9fb793ddd4be0f_2 _bdestructibleballoon_5_ec6de7164fe4891f6f812a886c5165cb_2 _modifymaxhealthclass_8_9267812849bafed41d45338dec7361ad_2 _setmaxhealthdatatag_11_fb51d36b4fc4780b1bf997a7195940fc_2 _fallspeed_14_17083faa4e3bb5af1f539a97bccffc27_2 _startlocked_17_5cdad18241633fc384b06495e1cdd4cf_2 _itemstospawn_58_af1cf0344c51389e47291b926b3c4812_2 _showflarewherelanding_25_151b3d8e4c220fbe2274a1b40a49a280_2 _showicononminimap_27_12ebf43d4acaa803646ba8af8299ea62_2 _balloonmesh_31_4d96f63c40c0c936fb1cfcbde79aa7e3_2 _cratemesh_33_5cae5b664f2a4bd62d00398fec7a173b_2 _spawnwithoutballoon_36_53abb48447dd0415121f489eec78ee95_2 _ballonstyle_39_3d850d344c919ec561d32b9c60f8386d_2) (Utilities|Struct|BreakSupplyDropSpawnerReplicationData (Variables|Default|GetSupplyDropSpawnerReplicationData)))
  (bind (_ballonhealth_59_d201d9ea409d428dcf9fb793ddd4be0f_3 _bdestructibleballoon_5_ec6de7164fe4891f6f812a886c5165cb_3 _modifymaxhealthclass_8_9267812849bafed41d45338dec7361ad_3 _setmaxhealthdatatag_11_fb51d36b4fc4780b1bf997a7195940fc_3) (Utilities|Struct|BreakSupplyDropSpawnerReplicationData (Variables|Default|GetSupplyDropSpawnerReplicationData)))
  (bind (_ballonhealth_59_d201d9ea409d428dcf9fb793ddd4be0f_4 _bdestructibleballoon_5_ec6de7164fe4891f6f812a886c5165cb_4 _modifymaxhealthclass_8_9267812849bafed41d45338dec7361ad_4 _setmaxhealthdatatag_11_fb51d36b4fc4780b1bf997a7195940fc_4 _fallspeed_14_17083faa4e3bb5af1f539a97bccffc27_3 _startlocked_17_5cdad18241633fc384b06495e1cdd4cf_3 _itemstospawn_58_af1cf0344c51389e47291b926b3c4812_3 _showflarewherelanding_25_151b3d8e4c220fbe2274a1b40a49a280_3 _showicononminimap_27_12ebf43d4acaa803646ba8af8299ea62_3 _balloonmesh_31_4d96f63c40c0c936fb1cfcbde79aa7e3_3 _cratemesh_33_5cae5b664f2a4bd62d00398fec7a173b_3 _spawnwithoutballoon_36_53abb48447dd0415121f489eec78ee95_3 _ballonstyle_39_3d850d344c919ec561d32b9c60f8386d_3 _owningteam_46_014c1604484bcc6f29a968b9c413445c_2 _owningclass_47_6647b8064362a113893a9a97a0d97ff2_2 _invertteamselection_52_3aaa82324a91316f9cfd6686b80aeaf2_2 _invertclassselection_53_3a7e099245e1b8a9cfa1279aa14d099c_2 _invertteamdamage_56_0e4c75564ea7325df8603698205ae789_2 _invertclassdamage_57_696644e645c78348bbf88096fd6b5f98_2 _instigatorteam_62_d0420e7c49553dc95a16be82f5995c39 _instigatorisvalid_65_16ac469c4561da213430cfbd067742da) (Utilities|Struct|BreakSupplyDropSpawnerReplicationData (Variables|Default|GetSupplyDropSpawnerReplicationData)))
  (bind _returnvalue (not _showflarewherelanding_25_151b3d8e4c220fbe2274a1b40a49a280_2))
  (bind _supplydropspawnerreplicationdata (Variables|Default|GetSupplyDropSpawnerReplicationData))
  (bind _fortminigameprogress (Variables|Default|GetFortMinigameProgress))
  ;[1 1 1] initialize supply drop from supply drop spawner
  (bind _ascreative_supply_drop_balloon (Utilities|Casting|CastToCreativeSupplyDropBalloon (Variables|/Script/Engine.ChildActorComponent|GetChildActor (Variables|Default|GetChildActor)))
    (:then
      (Variables|Default|SetSupplyDropBalloon _ascreative_supply_drop_balloon)
      (Class|FortMinigameLogicComponent|TryAttachtoMinigame _fortminigameprogress)
      (Initialization|Initialize (Variables|Default|GetCreative_UseRestriction_Damage) _fortminigameprogress 0 0 _owningteam_46_014c1604484bcc6f29a968b9c413445c _invertteamdamage_56_0e4c75564ea7325df8603698205ae789 _owningclass_47_6647b8064362a113893a9a97a0d97ff2 _invertclassdamage_57_696644e645c78348bbf88096fd6b5f98 false)
      (Variables|/CRD_SupplyDropSpawner/Actors/CreativeSupplyDropBalloon.CreativeSupplyDropBalloon_C|SetDamageRestriction (Variables|Default|GetCreative_UseRestriction_Damage) (Variables|Default|GetSupplyDropBalloon)))
    (:CastFailed
      _))
  (Initialization|Initialize (Variables|Default|GetCreative_UseRestriction_Component) (Variables|Default|GetFortMinigameProgress) 0 0 _owningteam_46_014c1604484bcc6f29a968b9c413445c _invertteamselection_52_3aaa82324a91316f9cfd6686b80aeaf2 _owningclass_47_6647b8064362a113893a9a97a0d97ff2 _invertclassselection_53_3a7e099245e1b8a9cfa1279aa14d099c false)
  (CallFunction|UpdateUseRestrictionwithInstigator _owningteam_46_014c1604484bcc6f29a968b9c413445c_2 _invertteamselection_52_3aaa82324a91316f9cfd6686b80aeaf2_2 _instigatorteam_62_d0420e7c49553dc95a16be82f5995c39 _instigatorisvalid_65_16ac469c4561da213430cfbd067742da)
  (Networking|SwitchHasAuthority
    (:Authority
      (CallFunction|SetBalloonMaxHealth _bdestructibleballoon_5_ec6de7164fe4891f6f812a886c5165cb_3 _ballonhealth_59_d201d9ea409d428dcf9fb793ddd4be0f_3 _modifymaxhealthclass_8_9267812849bafed41d45338dec7361ad_3 _setmaxhealthdatatag_11_fb51d36b4fc4780b1bf997a7195940fc_3)
      (CallFunction|SetFallSpeed _spawnwithoutballoon_36_53abb48447dd0415121f489eec78ee95_2 _fallspeed_14_17083faa4e3bb5af1f539a97bccffc27_2)
      (Variables|Default|SetbLocked _startlocked_17_5cdad18241633fc384b06495e1cdd4cf_2)
      (CallFunction|SetSupplyDrop _itemstospawn_58_af1cf0344c51389e47291b926b3c4812_2))
    (:Remote
      _))
  (Development|SetHiddeninGame (Variables|Default|GetReticle_Niagara) _returnvalue true)
  ;;[1 1 1] Hide Smoke VFX and Disable Smoke Audio if VFX are set to hidden
  (CallFunction|MuteAudio _returnvalue)
  (MiniMapData|SetMiniMapIndicatorIsVisible (Variables|Default|GetFortMiniMap) _showicononminimap_27_12ebf43d4acaa803646ba8af8299ea62_2)
  (CallFunction|SetSupplyDropMesh _spawnwithoutballoon_36_53abb48447dd0415121f489eec78ee95_2 _balloonmesh_31_4d96f63c40c0c936fb1cfcbde79aa7e3_2 _cratemesh_33_5cae5b664f2a4bd62d00398fec7a173b_2 _ballonstyle_39_3d850d344c919ec561d32b9c60f8386d_2))

(fn OpenSupplyDrop
  (if (not (Variables|Default|GetOpened))
    (Class|DeviceSupplyDropSpawnerV2|ClearSpawnedSupplyDrop (Variables|Default|GetParentDevice))
    (|CallParentFunction)))

(fn SetBalloonMaxHealth
  (params (bDestructibleBalloon :type Boolean) (BalloonHealth :type Float%20%28double-precision%29) (ModifyMaxHealthClass :type Gameplay%20Effect%20Class%20Reference) (SetMaxHealthDataTag :type Gameplay%20Tag%20Structure))
  (bind _returnvalue (Ability|GetAbilitySystemComponent (Variables|Default|GetSupplyDropBalloon)))
  (bind _supplydropballoon (Variables|Default|GetSupplyDropBalloon))
  (Networking|SwitchHasAuthority
    (:Authority
      (Utilities|IsValid _supplydropballoon
        (:Is%20Valid
          (if (not (Variables|GetbDestructibleBalloon))
            (LocalVariables|SetLocalNewMaxHealth 1.0)
            (Fortnite|SetCanbeDamaged _supplydropballoon)
            (else
              (LocalVariables|SetLocalNewMaxHealth (Variables|GetBalloonHealth))))
          (bind _returnvalue_2 (Ability|GameplayEffect|AssignTagSetbyCallerMagnitude (GameplayEffects|MakeOutgoingSpec :self (Ability|GetAbilitySystemComponent _supplydropballoon) :GameplayEffectClass (Variables|GetModifyMaxHealthClass) :Level 0.0) (Variables|GetSetMaxHealthDataTag) (LocalVariables|GetLocalNewMaxHealth)))
          (Utilities|IsValid _returnvalue
            (:Is%20Valid
              (GameplayEffects|RemoveActiveGameplayEffect _returnvalue (Variables|Default|GetMaxHealthEffectHandlerSupplyDrop) -1)
              (bind _returnvalue_3 (GameplayEffects|ApplyGameplayEffectSpecToSelf _returnvalue _returnvalue_2))
              (Variables|Default|SetMaxHealthEffectHandlerSupplyDrop _returnvalue_3))))))))

(fn SetFallSpeed
  (params (SpawnWithoutBalloon :type Boolean) (SpawnerFallSpeed :type Float%20%28double-precision%29))
  (bind _supplydropballoon (Variables|Default|GetSupplyDropBalloon))
  (Networking|SwitchHasAuthority
    (:Authority
      (if SpawnWithoutBalloon
        (Utilities|IsValid _supplydropballoon
          (:Is%20Valid
            (Actor|SetLifeSpan _supplydropballoon 0.1)))
        (else
          (Variables|Default|SetFallSpeed (* (Variables|Default|GetFallSpeed) (Variables|GetSpawnerFallSpeed))))))))

(fn SetSupplyDrop
  (params (ItemsToSpawn :type Array%20of%20Fort%20World%20Item%20Definition%20Object%20References))
  (Networking|SwitchHasAuthority
    (:Authority
      (if (Utilities|Array|IsNotEmpty ItemsToSpawn)
        (Variables|Default|SetItemsToSpawn ItemsToSpawn)))))

(fn ShouldDie
  (params (Damage :type Float%20%28single-precision%29) (EventInstigator :type Controller%20Object%20Reference) (DamageCauser :type Actor%20Object%20Reference) (bIsTest :type Boolean) (EffectContext :type Gameplay%20Effect%20Context%20Handle%20Structure%20%28by%20ref%29))
  (bind _controller (CreativeAI|Controller|GetParticipantControllerfromActor EventInstigator false))
  (Utilities|IsValid EventInstigator
    (:Is%20Valid
      (Utilities|IsValid _controller
        (:Is%20Valid
          (return (CreativeUserRestrictions|CanParticipantControllerInteract (Variables|Default|GetCreative_UseRestriction_Damage) _controller false true false)))
        (:Is%20Not%20Valid
          _)))
    (:Is%20Not%20Valid
      _))
  (return true))

(fn SpawnLoot
  (if (Utilities|Array|IsEmpty (Variables|Default|GetItemsToSpawn))
    (bind _output_get (Variables|Default|SetLootTableNameToUse (Variables|Default|GetLootTableName)))
    (bind (_outloottodrop _returnvalue) (Loot|PickLootDrops _output_get (Loot|GetLootLevel) -1 "(GameplayTags=)"))
    (if _returnvalue
      (Variables|Default|SetDelayedSpawnLoot _outloottodrop)
      (CallFunction|OnDelayedSpawnLoot))
    (else
      (CallFunction|DelaySpawnLootfromItemComponent))))

(fn UpdateUseRestrictionWithInstigator
  (params (OwningTeam :type Creative%20Team%20Option%20Structure) (InvertTeamSelection :type Boolean) (InstigatorTeam :type Byte) (InstigatorIsValid :type Boolean))
  ;[0 1 0.09829 0.5] If the instigator should be the owner, call SetTeamsWithAffects with the instigator's team. This makes it so only members of the instigator's team can open supply drops.
  (if (Creative|TeamOption|IsTeamType OwningTeam "ActivatingTeam")
    (if (Variables|GetInstigatorIsValid)
      (TeamRestrictions|SetTeamswithAffects (Variables|Default|GetCreative_UseRestriction_Component) (Knobs|Team|ReshiftTeamIndexValue (Math|Conversions|ToInteger(Byte) (Variables|GetInstigatorTeam))) (Variables|GetInvertTeamSelection))
      (else
        (Development|PrintString "CRD_SupplyDropSpawner_V2::UpdateUseRestrictionWithInstigator Warning: Instigator is NOT valid." true true "(R=1.000000,G=0.788867,B=0.000000,A=1.000000)" 1e+01)))))

(fn ConstructionScript
  (|CallParentFunction))
