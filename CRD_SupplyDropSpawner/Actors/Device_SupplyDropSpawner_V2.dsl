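; Creative supply drop spawner device that spawns a configurable balloon crate on a minigame-start timer or event, with event-driven lock, open and destroy.
; Spawn points use the safe zone in Creative Royale, otherwise a radius around device or island, with sphere-trace checks, delayed offset retries and a device-location fallback.
; Pickups dropped on it before game start become crate contents, passed with user options as replication data.
; It also migrates legacy options and clears drops at game end.
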(event EventBeginPlay)

(event MinigameEndedDelegate(FortMinigameProgress)
  ;[1 1 1] Handle Minigame Start/End
  (Utilities|Time|ClearandInvalidateTimerbyHandle (Variables|Default|GetSpawnDelayTimer))
  (CallFunction|DestroySpawnedSupplyDrops)
  (CallFunction|ResetSpawnAttemptGate))

(event MinigameStartedDelegate(FortMinigameProgress)
  ;[1 1 1] Handle Minigame Start/End
  (Networking|SwitchHasAuthority
    (:Authority
      (switch Utilities|FlowControl|Switch|SwitchonCRD_SupplyDropSpawner_SpawnDelay (Variables|UserOptions|Basic|GetSpawnDelay)
        (:NewEnumerator0
          (Variables|UserOptions|Basic|SetCustomSpawnDelay 0.0))
        (:NewEnumerator1
          (Variables|UserOptions|Basic|SetCustomSpawnDelay 1.0))
        (:NewEnumerator2
          _))
      (bind _returnvalue (Utilities|Time|SetTimerbyEvent (EventDispatchers|CreateEvent :!event Custom|SpawnSupplyDrop) (Variables|UserOptions|Basic|GetCustomSpawnDelay)))
      (Variables|Default|SetSpawnDelayTimer _returnvalue)
      (Variables|Default|SetSpawnAttemptsRemaining (Variables|Default|GetSpawnAttempts)))))

(event OnComponentHit(PickupCollisionBox)
  (params (HitComponent :type Primitive%20Component%20Object%20Reference) (OtherActor :type Actor%20Object%20Reference) (OtherComp :type Primitive%20Component%20Object%20Reference) (NormalImpulse :type Vector :default "0, 0, 0") (Hit :type Hit%20Result%20Structure%20%28by%20ref%29))
  ;[1 1 1] Handle Item Drops
  (Networking|SwitchHasAuthority
    (:Authority
      (Utilities|IsValid OtherActor
        (:Is%20Valid
          (if (not (Minigame|HasStarted (Variables|/Script/FortniteGame.FortMinigameLogicComponent|GetCurrentMinigame (Variables|Default|GetFortMinigameProgress)) :!class /Script/FortniteGame.FortMinigame))
            (CallFunction|HandleItemDropped OtherActor)))))))

(event Custom|Multicast_AddItemFX
  (params (Item :type Fort%20Pickup%20Athena%20Object%20Reference))
  ;[1 1 1] Handle Item Drops
  (Networking|CanExecuteCosmeticEvents
    (:True
      (Class|CreativeItemPreviewComponent|ItemAddedToQueueFX (Variables|Default|GetCreative_ItemPreview_Component) Item))))

(event EventInitializeDevice
  (|CallParentFunction)
  (Initialization|InitializeComponent (Variables|Default|GetCreative_ItemPreview_Component) (Variables|Default|GetPreviewMeshScene) (Variables|Default|GetPreviewStaticMesh) (Variables|Default|GetPreviewSkeletalMesh) 0 (Variables|Default|GetPickupItemListComponent) 0 :!class /Game/Creative/Devices/Common/Components/Creative_ItemPreview_Component.Creative_ItemPreview_Component_C)
  (Initialization|Initialize (Variables|Default|GetCreative_UseRestriction_Component) (Variables|Default|GetFortMinigameProgress) 0 0 (Variables|UserOptions|Basic|GetOwningTeam) (Variables|UserOptions|Advanced|GetInvertTeamSelection) (Variables|UserOptions|Basic|GetOwningClass) (Variables|UserOptions|Advanced|GetInvertClassSelection) false)
  (Initialization|Initialize (Variables|Default|GetCreative_UseRestriction_Damage) (Variables|Default|GetFortMinigameProgress) 0 0 (Variables|UserOptions|Basic|GetOwningTeam) (Variables|UserOptions|Advanced|GetInvertTeamDamage) (Variables|UserOptions|Basic|GetOwningClass) (Variables|UserOptions|Advanced|GetInvertClassDamage) false)
  (Variables|Default|SetSupplyDropSpawning false))

(event Custom|SpawnSupplyDrop
  (bind (_spawnlocation _bsuccess) (CallFunction|CalculateSpawnLocation))
  (bind _spawnattemptsremaining (Variables|Default|GetSpawnAttemptsRemaining))
  ;[0 1 0.3421 0.5] Spawn Supply Drop
  (Networking|SwitchHasAuthority
    (:Authority
      (if _bsuccess
        (bind _output_get (Variables|Default|SetSpawnLocation _spawnlocation))
        (if (CallFunction|CanSpawnSupplyDrop _output_get)
          _
          (elif (<= _spawnattemptsremaining 0)
            (Utilities|FlowControl|DoOnce false
              (:Completed
                (bind _output_get_2 (Variables|Default|SetSpawnLocation (Transformation|GetActorLocation)))
                (if (CallFunction|CanSpawnSupplyDrop _output_get_2 true)
                  (Variables|Default|SetForcedSpawnAtLocation true)
                  (bind _returnvalue (Building|SpawnBuildingActor (Utilities|Map|Find (Variables|Default|GetSpawnedClasses) (Variables|UserOptions|Basic|GetBalloonStyle)) (Math|Transform|MakeTransform (Math|Vector|SelectVector (Variables|Default|GetSpawnLocation) (Utilities|Operators|Add (Variables|Default|GetSpawnLocation) (Math|Vector|MakeVector (Variables|Default|GetSpawnLocationModifier) (Variables|Default|GetSpawnLocationModifier))) (Variables|Default|GetForcedSpawnAtLocation))) self (CallFunction|GetSupplyDropInstigatorPawn) false false))
                  (bind _asbpi_creative_supply_drop (Utilities|Casting|CastToBPI_CreativeSupplyDrop _returnvalue))
                  (bind _output_get_3 (Variables|Default|SetSpawnedSupplyDrop _asbpi_creative_supply_drop))
                  (Utilities|Array|Add (Variables|Default|GetRemainingSpawnedDrops) _output_get_3)
                  (Class|BPICreativeSupplyDrop|SetParentDevice _output_get_3 self)
                  (bind _replicationdata (CallFunction|GetSupplyDropReplicationData))
                  (Class|BPICreativeSupplyDrop|SetReplicationData _output_get_3 _replicationdata)
                  (Class|BPICreativeSupplyDrop|SetFlareColor _output_get_3 (Variables|UserOptions|Basic|GetSupplyFXColor))
                  (else
                    _))
                (Variables|Default|SetSpawnLocationModifier 0.0)
                (Variables|Default|SetSpawnAttemptsRemaining (Variables|Default|GetSpawnAttempts))
                (Variables|Default|SetSupplyDropSpawning false)
                (Variables|Default|SetForcedSpawnAtLocation false)
                (CallFunction|ResetSpawnAttemptGate)))
            (else
              (Math|Integer|DecrementInt _spawnattemptsremaining
                (:%20%20
                  (Variables|Default|SetSpawnLocationModifier (+ (Variables|Default|GetSpawnLocationModifier) 6.5e+02))
                  (Utilities|FlowControl|Delay 1.0)
                  (CallFunction|SpawnSupplyDrop))))))
        (else
          (Development|PrintString "CRD_SupplyDropSpawner::SpawnSupplyDrop Error: Unable to find valid spawn location." true true "(R=1.000000,G=0.000000,B=0.069327,A=1.000000)" 1e+01))))))

(event Custom|ResetSpawnAttemptGate
  ;[0 1 0.3421 0.5] Spawn Supply Drop
  (Utilities|FlowControl|DoOnce false
    (:Completed
      (bind _output_get (Variables|Default|SetSpawnLocation (Transformation|GetActorLocation)))
      (if (CallFunction|CanSpawnSupplyDrop _output_get true)
        (Variables|Default|SetForcedSpawnAtLocation true)
        (bind _returnvalue (Building|SpawnBuildingActor (Utilities|Map|Find (Variables|Default|GetSpawnedClasses) (Variables|UserOptions|Basic|GetBalloonStyle)) (Math|Transform|MakeTransform (Math|Vector|SelectVector (Variables|Default|GetSpawnLocation) (Utilities|Operators|Add (Variables|Default|GetSpawnLocation) (Math|Vector|MakeVector (Variables|Default|GetSpawnLocationModifier) (Variables|Default|GetSpawnLocationModifier))) (Variables|Default|GetForcedSpawnAtLocation))) self (CallFunction|GetSupplyDropInstigatorPawn) false false))
        (bind _asbpi_creative_supply_drop (Utilities|Casting|CastToBPI_CreativeSupplyDrop _returnvalue))
        (bind _output_get_2 (Variables|Default|SetSpawnedSupplyDrop _asbpi_creative_supply_drop))
        (Utilities|Array|Add (Variables|Default|GetRemainingSpawnedDrops) _output_get_2)
        (Class|BPICreativeSupplyDrop|SetParentDevice _output_get_2 self)
        (bind _replicationdata (CallFunction|GetSupplyDropReplicationData))
        (Class|BPICreativeSupplyDrop|SetReplicationData _output_get_2 _replicationdata)
        (Class|BPICreativeSupplyDrop|SetFlareColor _output_get_2 (Variables|UserOptions|Basic|GetSupplyFXColor))
        (else
          _))
      (Variables|Default|SetSpawnLocationModifier 0.0)
      (Variables|Default|SetSpawnAttemptsRemaining (Variables|Default|GetSpawnAttempts))
      (Variables|Default|SetSupplyDropSpawning false)
      (Variables|Default|SetForcedSpawnAtLocation false)
      (CallFunction|ResetSpawnAttemptGate))))

(event Custom|SpawnWhenReceivingEvent
  (params (Instigator :type Controller%20Object%20Reference))
  ;[0.1659 0.0612 0.4896] Triggers and Recievers
  (if (not (Variables|Default|GetSupplyDropSpawning))
    (Variables|Default|SetSupplyDropSpawning true)
    (if (Utilities|IsValidInterface (Variables|Default|GetSpawnedSupplyDrop))
      (Variables|Default|SetSupplyDropSpawning false)
      (else
        (Variables|Default|SetSupplyDropSpawnerInstigator Instigator)
        (Utilities|Time|ClearandInvalidateTimerbyHandle (Variables|Default|GetSpawnDelayTimer))
        (CallFunction|SpawnSupplyDrop)))))

(event Custom|DestroyWhenReceivingEvent
  (params (Instigator :type Controller%20Object%20Reference))
  (bind _spawnedsupplydrop (Variables|Default|GetSpawnedSupplyDrop))
  ;[0.1659 0.0612 0.4896] Triggers and Recievers
  (if (Utilities|IsValidInterface _spawnedsupplydrop)
    (Class|BPICreativeSupplyDrop|DestroyBalloon _spawnedsupplydrop)))

(event Custom|OpenSupplyDropWhenReceivingEvent
  (params (Instigator :type Controller%20Object%20Reference))
  (bind _spawnedsupplydrop (Variables|Default|GetSpawnedSupplyDrop))
  ;[0.1659 0.0612 0.4896] Triggers and Recievers
  (if (Utilities|IsValidInterface _spawnedsupplydrop)
    (Class|BPICreativeSupplyDrop|OpenCreativeSupplyDrop _spawnedsupplydrop Instigator)
    (GameplayEvent|BroadcastEvent (Variables|UserOptions|Events|GetOn%20Opened) 0 (Utilities|Struct|MakeGameplayEventSupplyDropSpawnerTransmitonOpen Instigator))))

(event Custom|UnlockWhenReceivingEvent
  (params (Instigator :type Controller%20Object%20Reference))
  ;[0.1659 0.0612 0.4896] Triggers and Recievers
  (Class|BPICreativeSupplyDrop|SetLocked (Variables|Default|GetSpawnedSupplyDrop) false))

(event Custom|DestroySpawnedDropsWhenReceivingEvent
  ;[0.1659 0.0612 0.4896] Triggers and Recievers
  (CallFunction|DestroySpawnedSupplyDrops))

(event Custom|LockWhenReceivingEvent
  (params (Instigator :type Controller%20Object%20Reference))
  ;[0.1659 0.0612 0.4896] Triggers and Recievers
  (Class|BPICreativeSupplyDrop|SetLocked (Variables|Default|GetSpawnedSupplyDrop) true))

(event Custom|HandleSupplyDropDestroyed
  (params (SupplyDrop :type BPI%20Creative%20Supply%20Drop%20Interface) (Instigator :type Controller%20Object%20Reference))
  (Utilities|Array|RemoveItem (Variables|Default|GetRemainingSpawnedDrops) SupplyDrop)
  (GameplayEvent|BroadcastEvent (Variables|UserOptions|Events|GetOn%20Crate%20Destroyed) 0 (Utilities|Struct|MakeGameplayEventSupplyDropSpawnerTransmitonDestroy Instigator)))

(fn AttemptSupplyDropSpawn
  (bind _returnvalue (Utility|OverlapQueue|GetAsyncOverlapQueue self))
  (Class|AsyncOverlapQueue|ThrottledAsyncSphereOverlapActors :self _returnvalue :Location (Utilities|Operators|Add (Transformation|GetActorLocation) (Math|Vector|MakeVector (Variables|Default|GetSpawnLocationModifier) (Variables|Default|GetSpawnLocationModifier) (Variables|Default|GetFallHeight))) :Radius 1e+04 :bCheckLineOfSight false :ShouldCancelProcess (EventDispatchers|CreateEvent :!event ThrottleSpawnCancelEvent) :ProcessHitActorFunction (EventDispatchers|CreateEvent :!event ThrottleSpawnProcessEvent) :CompletionFunction (EventDispatchers|CreateEvent :!event ThrottleSpawnCompletionEvent) :ObjectTypes "ObjectTypeQuery1" :TraceChannel "ECC_Visibility"))

(fn CalculateSpawnLocation
  (bind (_safezonelocation _bsuccess) (CallFunction|MakeSupplyDropLocationfromSafeZone))
  (bind _spawnradius (CallFunction|GetSpawnRadius))
  (bind _returnvalue (* _spawnradius -1.0))
  (bind _returnvalue_2 (Math|Random|RandomFloatinRange _returnvalue _spawnradius))
  ;[0 1 0.2245 0.5] Calculate the spawn location. Use the safe zone if Creative Royale.
  (if (CallFunction|IsCreativeRoyale)
    (return _safezonelocation _bsuccess)
    (else
      ;;[1 0.8798 0.5885 0.5] Save as local variables to force the calculation to run twice.
      (LocalVariables|SetRadomLocXLocal _returnvalue_2)
      (LocalVariables|SetRadomLocYLocal _returnvalue_2)
      (return (Utilities|Operators|Add (CallFunction|GetOriginforCalculations) (Math|Vector|MakeVector (LocalVariables|GetRadomLocXLocal) (LocalVariables|GetRadomLocYLocal))) true))))

(fn CanSpawnSupplyDrop
  (params (Start :type Vector) (ForceSpawnAtDeviceLocation :type Boolean))
  (bind _fallheight (Variables|Default|GetFallHeight))
  (bind _spawnlocationmodifier (Variables|Default|GetSpawnLocationModifier))
  (bind _returnvalue (Utilities|Operators|Add Start (Math|Vector|MakeVector _spawnlocationmodifier _spawnlocationmodifier _fallheight)))
  (bind _fallheight_2 (Variables|Default|GetFallHeight))
  (bind (_returnvalue_x _returnvalue_y) (Math|Vector|BreakVector (Transformation|GetActorLocation)))
  (bind _returnvalue_2 (Utilities|Operators|Add Start (Math|Vector|MakeVector _returnvalue_x _returnvalue_y _fallheight_2)))
  (bind _returnvalue_3 (Math|Vector|SelectVector _returnvalue_2 _returnvalue ForceSpawnAtDeviceLocation))
  (Utilities|IsLocationRelevanttoMe _returnvalue_3
    (:IsRelevant
      (bind (_outhits _returnvalue_4) (Collision|MultiSphereTraceByChannel _returnvalue_3 _returnvalue_3 1024.0))
      (if _returnvalue_4
        (return false)
        (else
          (return true))))
    (:NotRelevant
      (return false))))

(fn ClearSpawnedSupplyDrop
  (Variables|Default|SetSpawnedSupplyDrop 0))

(fn DestroySpawnedSupplyDrops
  (bind _spawnedsupplydrop (Variables|Default|GetSpawnedSupplyDrop))
  (for _array_element (Variables|Default|GetRemainingSpawnedDrops)
    (Class|BPICreativeSupplyDrop|DestroyCreativeSupplyDrop _array_element))
  (Utilities|Array|Clear (Variables|Default|GetRemainingSpawnedDrops))
  (if (Utilities|IsValidInterface _spawnedsupplydrop)
    (Class|BPICreativeSupplyDrop|DestroyCreativeSupplyDrop _spawnedsupplydrop)
    (Variables|Default|SetSpawnedSupplyDrop 0)))

(fn GetMinigameVolumeMesh
  (bind _self self)
  (bind _returnvalue (Minigame|GetMinigame _self :!class /Script/FortniteGame.MinigameBlueprintLibrary))
  ;[0 1 0.2101 0.5] Get the Minigame volume's mesh.
  (Utilities|IsValid _returnvalue
    (:Is%20Valid
      (bind _asfort_volume (Utilities|Casting|CastToFortVolume (Variables|/Script/FortniteGame.FortMinigame|GetVolume _returnvalue))
        (:then
          (return (Volume|GetStaticMesh _asfort_volume)))
        (:CastFailed
          (Development|PrintString "CRD_SupplyDropSpawner::GetMinigameVolumeMesh Error: Cast to FortVolume failed." true true "(R=1.000000,G=0.015783,B=0.000000,A=1.000000)" 1e+01)
          (return 0))))
    (:Is%20Not%20Valid
      (Development|PrintString "CRD_SupplyDropSpawner::GetMinigameVolumeMesh Error: Miningame is NOT valid." true true "(R=1.000000,G=0.015783,B=0.000000,A=1.000000)" 1e+01)
      (return 0))))

(fn GetOriginForCalculations
  (bind _minigamevolumemesh (CallFunction|GetMinigameVolumeMesh))
  (bind _origin (Collision|GetComponentBounds _minigamevolumemesh))
  (bind (_origin_x _origin_y) (Math|Vector|BreakVector _origin))
  (bind _returnvalue (Transformation|GetActorLocation))
  (bind (_returnvalue_x _returnvalue_y _returnvalue_z) (Math|Vector|BreakVector (Transformation|GetActorLocation)))
  ;[0 1 0.3424 0.5] If the Spawn Radius is -1, use the Minigame volume mesh origin; otherwise, use the actor's location as the origin.
  (switch Utilities|FlowControl|Switch|SwitchonCRD_SupplyDropSpawner_SpawnRadius (Variables|UserOptions|Basic|GetSpawnRadius)
    (:NewEnumerator0
      (Utilities|IsValid _minigamevolumemesh
        (:Is%20Valid
          (return _origin_x _origin_y _returnvalue_z))
        (:Is%20Not%20Valid
          (Development|PrintString "CRD_SupplyDropSpawner::SpawnSupplyDrop Error: Miningame Volume Mesh is NOT valid." true true "(R=1.000000,G=0.015783,B=0.000000,A=1.000000)" 1e+01)
          (return _returnvalue))))
    (:NewEnumerator1
      _)
    (:NewEnumerator2
      _))
  (return _returnvalue))

(fn GetPlayableIslandRadius
  (bind _safezoneindicator (CallFunction|GetSafeZoneIndicator))
  (bind _minigamevolumemesh (CallFunction|GetMinigameVolumeMesh))
  (bind (_origin _boxextent) (Collision|GetComponentBounds _minigamevolumemesh))
  (bind (_boxextent_x _boxextent_y) (Math|Vector|BreakVector _boxextent))
  ;[0 1 0.2829 0.5] Gets the radius of the playable region of the map.
  (if (CallFunction|IsCreativeRoyale)
    ;;[1 0.9818 0.3646 0.5] Use the safe zone for Creative Royale.
    (Utilities|IsValid _safezoneindicator
      (:Is%20Valid
        (return (SafeZone|GetSafeZoneRadius _safezoneindicator)))
      (:Is%20Not%20Valid
        (Development|PrintString "CRD_SupplyDropSpawner::GetIslandRadus Error: SafeZoneIndicator is NOT valid." true true "(R=1.000000,G=0.015783,B=0.000000,A=1.000000)" 1e+01)
        (return 0.0)))
    (else
      ;;[1 0.7919 0 0.5] Get the minimum extent of the Minigame's volume mesh for non-Creative Royale maps since there is no safe zone.
      (Utilities|IsValid _minigamevolumemesh
        (:Is%20Valid
          (return (Math|Float|Min(Float) _boxextent_x _boxextent_y)))
        (:Is%20Not%20Valid
          (Development|PrintString "CRD_SupplyDropSpawner::GetIslandRadus Error: Miningame Volume Mesh is NOT valid." true true "(R=1.000000,G=0.015783,B=0.000000,A=1.000000)" 1e+01)
          (return 0.0))))))

(fn GetSafeZoneCenterFromSafeZonePhases
  (bind (_outsafezonephase _returnvalue) (SafeZone|GetPhaseInfo (SafeZone|GetSafeZoneIndicator) -1))
  (bind _returnvalue_2 (SafeZone|GetSafeZoneIndicator))
  ;[0 1 0.02015 0.5] Gets the current safe zone center. Getting the safe zone center by calling GetSafeZoneCenter from the SafeZoneIndicator was returning the mesh's location which has a different Z value that is too high. The high Z value was making it so the supply drop could not detect the ground and would drift toward map center.
  (Utilities|FlowControl|Sequence
    (:then_0
      (Utilities|IsValid _returnvalue_2
        (:Is%20Valid
          (if _returnvalue
            (return (Utilities|Struct|BreakFortSafeZonePhaseInfo _outsafezonephase))
            (else
              (Development|PrintString "CRD_SupplyDropSpawner::GetSafeZoneCenterFromSafeZonePhases Error: GetPhaseInfo failed." true true "(R=1.000000,G=0.022075,B=0.000000,A=1.000000)" 5.0))))
        (:Is%20Not%20Valid
          (Development|PrintString "CRD_SupplyDropSpawner::GetSafeZoneCenterFromSafeZonePhases Error: Safe zone indicator is NOT valid." true true "(R=1.000000,G=0.022075,B=0.000000,A=1.000000)" 5.0))))
    (:then_1
      (Development|PrintString "CRD_SupplyDropSpawner::GetSafeZoneCenterFromSafeZonePhases Error: Unable to get safe zone center." true true "(R=1.000000,G=0.022075,B=0.000000,A=1.000000)" 5.0)
      (return "0, 0, 0"))))

(fn GetSafeZoneIndicator
  (bind _returnvalue (SafeZone|GetSafeZoneIndicator))
  ;[0 1 0.1947 0.5] Get the safe zone indicator.
  (Utilities|IsValid _returnvalue
    (:Is%20Valid
      (return _returnvalue))
    (:Is%20Not%20Valid
      (Development|PrintString "CRD_SupplyDropSpawner::GetSafeZoneIndicator SafeZoneIndicator is NOT valid." true true "(R=1.000000,G=0.042821,B=0.000000,A=1.000000)" 1e+01)
      (return 0))))

(fn GetSpawnRadius
  ;[0 1 0.1877 0.5] If the Spawn Radius is -1, use the playable region of the island; otherwise, use the variable's value.
  (switch Utilities|FlowControl|Switch|SwitchonCRD_SupplyDropSpawner_SpawnRadius (Variables|UserOptions|Basic|GetSpawnRadius)
    (:NewEnumerator0
      (return (CallFunction|GetPlayableIslandRadius)))
    (:NewEnumerator1
      (return 0.0))
    (:NewEnumerator2
      (return (* (Variables|UserOptions|Basic|GetCustomSpawnRadius) 1e+02)))))

(fn GetSupplyDropInstigatorPawn
  ;[0 1 0.3331 0.5] Gets the controlled pawn from the instigator.
  (bind _supplydropspawnerinstigator (Variables|Default|GetSupplyDropSpawnerInstigator)
    (:then
      (return (Pawn|GetControlledPawn _supplydropspawnerinstigator)))
    (:else
      (return 0))))

(fn GetSupplyDropReplicationData
  (return (Utilities|Struct|MakeSupplyDropSpawnerReplicationData (Variables|UserOptions|Basic|GetBalloonHealth) (Variables|UserOptions|Basic|GetbDestructibleBalloon) (Variables|Constants|GetModifyMaxHealth) (Variables|Default|GetSetMaxHealthDataTag) (Variables|UserOptions|Basic|GetFallSpeed) (Variables|UserOptions|Advanced|GetStartLocked) (ItemData|GetContainedItems (Variables|Default|GetPickupItemListComponent)) (Variables|UserOptions|Advanced|GetShowFlareWhereLanding) (Variables|UserOptions|Advanced|GetShowIconOnMinimap) (Utilities|Map|Find (Variables|Default|GetBalloonMeshes) (Variables|UserOptions|Basic|GetBalloonStyle)) (Utilities|Map|Find (Variables|Default|GetCrateMeshes) (Variables|UserOptions|Basic|GetBalloonStyle)) (Variables|UserOptions|Basic|GetSpawnWithoutBalloon) (Variables|UserOptions|Basic|GetBalloonStyle) (Variables|UserOptions|Basic|GetOwningTeam) (Variables|UserOptions|Basic|GetOwningClass) (Variables|UserOptions|Advanced|GetInvertTeamSelection) (Variables|UserOptions|Advanced|GetInvertClassSelection) (Variables|UserOptions|Advanced|GetInvertTeamDamage) (Variables|UserOptions|Advanced|GetInvertClassDamage) (Team|GetActorPvPTeamIndex (CallFunction|GetSupplyDropInstigatorPawn)) (Utilities|IsValid (CallFunction|GetSupplyDropInstigatorPawn) :!class /Script/Engine.KismetSystemLibrary))))

(fn HandleItemDropped
  (params (Actor :type Actor%20Object%20Reference))
  (bind _asfort_pickup_athena (Utilities|Casting|CastToFortPickupAthena Actor))
  (bind _output_get (LocalVariables|SetPickupAthena _asfort_pickup_athena))
  (Utilities|IsPickupValid _output_get (Variables|/Game/Creative/Devices/CaptureArea/PickupItemListComponent.PickupItemListComponent_C|GetInvalidItemTags (Variables|Default|GetPickupItemListComponent))
    (:IsValid
      (if (not (Class|PickupItemListComponent|IsIteminQueue (Variables|Default|GetPickupItemListComponent) (LocalVariables|GetPickupAthena)))
        (bind _bitemadded (Class|PickupItemListComponent|AddItemtoQueue (Variables|Default|GetPickupItemListComponent) (LocalVariables|GetPickupAthena)))
        (if _bitemadded
          (CallFunction|MulticastAddItemFX (LocalVariables|GetPickupAthena)))))))

(fn IsCreativeRoyale
  (bind _returnvalue (Fortnite|Athena|GameState|GetGameStateAthena))
  ;[0 1 0.2074 0.5] If this playlist has a Creative Royale tag return true.
  (Utilities|IsValid _returnvalue
    (:Is%20Valid
      (return (GameplayTags|HasTag (Class|FortGameStateAthena|GetAthenaPlaylistContextTags _returnvalue) "(TagName=\"Athena.Playlist.CreativeRoyale\")")))
    (:Is%20Not%20Valid
      (return false))))

(fn MakeSupplyDropLocationFromSafeZone
  (bind _returnvalue (Fortnite|Athena|GameState|GetGameStateAthena))
  (bind _mapinfo (Variables|/Script/FortniteGame.FortGameStateAthena|GetMapInfo _returnvalue))
  ;[0 1 0.2284 0.5] Creates a supply drop spawn location using the safe zone as the origin. Water will be avoided.
  (Utilities|IsValid _returnvalue
    (:Is%20Valid
      (Utilities|IsValid _mapinfo
        (:Is%20Valid
          (bind _center (CallFunction|GetSafeZoneCenterfromSafeZonePhases))
          (return (Class|FortAthenaMapInfo|PickSupplyDropLocation _mapinfo _center (CallFunction|GetSpawnRadius) true (ScalableFloat|GetValueatLevel (Variables|/Script/FortniteGame.FortAthenaMapInfo|GetSupplyDropMaxPlacementHeight _mapinfo)) (ScalableFloat|GetValueatLevel (Variables|/Script/FortniteGame.FortAthenaMapInfo|GetSupplyDropMinPlacementHeight _mapinfo)) false) true))
        (:Is%20Not%20Valid
          _)))
    (:Is%20Not%20Valid
      _))
  (return "0, 0, 0" false))

(fn MigrateLegacyUserOptions
  (bind _fballoonhealth (Class|GetClassDefaults "/CRD_SupplyDropSpawner/Actors/Device_SupplyDropSpawner_V2.Device_SupplyDropSpawner_V2_C"))
  (bind _ffallspeedmultiplier (Class|GetClassDefaults "/CRD_SupplyDropSpawner/Actors/Device_SupplyDropSpawner_V2.Device_SupplyDropSpawner_V2_C"))
  (bind _fspawndelay (Class|GetClassDefaults "/CRD_SupplyDropSpawner/Actors/Device_SupplyDropSpawner_V2.Device_SupplyDropSpawner_V2_C"))
  (bind _fspawnradius (Class|GetClassDefaults "/CRD_SupplyDropSpawner/Actors/Device_SupplyDropSpawner_V2.Device_SupplyDropSpawner_V2_C"))
  (bind _fspawnradius_2 (Variables|Legacy|GetfSpawnRadius))
  (bind _owningclassindexsetting (Variables|Legacy|GetOwningClassIndexSetting))
  (|CallParentFunction)
  (Utilities|FlowControl|Sequence
    (:then_0
      (Utilities|FlowControl|Sequence
        (:then_0
          ;[0 0.5451 1 0.502] Migrating fSpawnDelay -> Splitting into SpawnDelay (Off, Game Start, Custom), CustomSpawnDelay (float)
          (if (!= _fspawndelay (Variables|Legacy|GetfSpawnDelay))
            (if (!= (Variables|Legacy|GetfSpawnDelay) -1.0)
              (Variables|UserOptions|Basic|SetSpawnDelay "NewEnumerator2")
              (Variables|UserOptions|Basic|SetCustomSpawnDelay (Variables|Legacy|GetfSpawnDelay))
              (else
                (Variables|UserOptions|Basic|SetSpawnDelay "NewEnumerator0")))
            (Variables|Legacy|SetfSpawnDelay _fspawndelay)
            (LocalVariables|SetMigratedProperties true)))
        (:then_1
          ;[0 1 0.2824 0.502] Migrating fSpawnRadius -> Splitting into SpawnRadius (Island,Above,Custom), CustomSpawnRadius (float)
          (if (!= _fspawnradius (Variables|Legacy|GetfSpawnRadius))
            (if (== _fspawnradius_2 -1.0)
              (Variables|UserOptions|Basic|SetSpawnRadius "NewEnumerator0")
              (elif (> _fspawnradius_2 0)
                (Variables|UserOptions|Basic|SetSpawnRadius "NewEnumerator2")
                (Variables|UserOptions|Basic|SetCustomSpawnRadius (/ (Variables|Legacy|GetfSpawnRadius) 1e+02))
                (else
                  (Variables|UserOptions|Basic|SetSpawnRadius "NewEnumerator1"))))
            (Variables|Legacy|SetfSpawnRadius _fspawnradius)
            (LocalVariables|SetMigratedProperties true)))
        (:then_2
          ;[0 1 0.8706 0.502] Migrating fBalloonHealth -> Splitting into bDestructibleBalloon (Bool), BalloonHealth (float)
          (if (!= _fballoonhealth (Variables|Legacy|GetfBalloonHealth))
            (if (!= (Variables|Legacy|GetfBalloonHealth) -1.0)
              (Variables|UserOptions|Basic|SetbDestructibleBalloon true)
              (Variables|UserOptions|Basic|SetBalloonHealth (Variables|Legacy|GetfBalloonHealth))
              (else
                (Variables|UserOptions|Basic|SetbDestructibleBalloon false)))
            (Variables|Legacy|SetfBalloonHealth _fballoonhealth)
            (LocalVariables|SetMigratedProperties true)))
        (:then_3
          ;[0 0.3111 1 0.502] Migrating fFallSpeed -> Splitting into SpawnWithoutBalloon (Bool), FallSpeed (float)
          (if (!= _ffallspeedmultiplier (Variables|Legacy|GetfFallSpeedMultiplier))
            (if (!= (Variables|Legacy|GetfFallSpeedMultiplier) 0)
              (Variables|UserOptions|Basic|SetSpawnWithoutBalloon false)
              (Variables|UserOptions|Basic|SetFallSpeed (Variables|Legacy|GetfFallSpeedMultiplier))
              (else
                (Variables|UserOptions|Basic|SetSpawnWithoutBalloon true)))
            (Variables|Legacy|SetfFallSpeedMultiplier _ffallspeedmultiplier)
            (LocalVariables|SetMigratedProperties true)))
        (:then_4
          ;[1 0 0.6271 0.25] Migrating Invalid Team / Class Settings into value ones.
          (if (!= (Class|GetClassDefaults "/CRD_SupplyDropSpawner/Actors/Device_SupplyDropSpawner_V2.Device_SupplyDropSpawner_V2_C") (Variables|Legacy|GetOwningTeamIndexSetting))
            (if (== (Variables|Legacy|GetOwningTeamIndexSetting) -999)
              (Variables|SetOwningTeam_TeamType (struct :TeamType "ActivatingTeam" :TeamIndex 1))
              (else
                (Variables|UserOptions|Basic|SetOwningTeam (Creative|TeamOption|MakefromLegacyTeamIndex (Variables|Legacy|GetOwningTeamIndexSetting)))))
            (Variables|Legacy|SetOwningTeamIndexSetting (Class|GetClassDefaults "/CRD_SupplyDropSpawner/Actors/Device_SupplyDropSpawner_V2.Device_SupplyDropSpawner_V2_C"))
            (LocalVariables|SetMigratedProperties true)))
        (:then_5
          (if (!= (Class|GetClassDefaults "/CRD_SupplyDropSpawner/Actors/Device_SupplyDropSpawner_V2.Device_SupplyDropSpawner_V2_C") _owningclassindexsetting)
            (switch int _owningclassindexsetting
              (:Default
                (Creative|ClassOption|SetLegacyClassIndex (Variables|UserOptions|Basic|GetOwningClass) _owningclassindexsetting))
              (:-2
                (Variables|SetOwningClass_ClassType (struct :ClassType "NoClass" :ClassSlot 1))))
            (Variables|Legacy|SetOwningClassIndexSetting (Class|GetClassDefaults "/CRD_SupplyDropSpawner/Actors/Device_SupplyDropSpawner_V2.Device_SupplyDropSpawner_V2_C"))
            (LocalVariables|SetMigratedProperties true)))))
    (:then_1
      (if (LocalVariables|GetMigratedProperties)
        (OptionData|RefreshPropertyOverrides (Variables|Default|GetToyOptionsComponent))))))

(fn ThrottleSpawnCancelEvent
  (return false))

(fn ThrottleSpawnCompletionEvent
  (params (WasCancelled :type Boolean))
  (bind _spawnattempts (Variables|Default|GetSpawnAttempts))
  (if (Variables|Default|GetAllowedToSpawn)
    (CallFunction|SpawnSupplyDrop)
    (elif (> _spawnattempts 0)
      (Math|Integer|DecrementInt _spawnattempts
        (:%20%20
          (Variables|Default|SetSpawnLocationModifier (+ (Variables|Default|GetSpawnLocationModifier) 5e+02))
          (Variables|Default|SetAllowedToSpawn true)
          (CallFunction|AttemptSupplyDropSpawn))))))

(fn ThrottleSpawnProcessEvent
  (params (HitActor :type Actor%20Object%20Reference))
  (if (Actor|ActorHasTag HitActor "Creative.SupplyDrop")
    (Variables|Default|SetAllowedToSpawn false)))

(fn ConstructionScript
  (|CallParentFunction))
